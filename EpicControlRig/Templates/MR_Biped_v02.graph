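# MR_Biped_v02 is a biped Control Rig template in EpicControlRig, holding a large rig hierarchy.
# Its RigVM graph has only a Begin Execution node and its function library is empty, so those graphs hold no solve logic.
# The visible hierarchy declares animation curves for joint corrective poses, facial poses and CTRL_expressions, plus gameplay controls for IK, additives, warping, locomotion, weapons and emotes.
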ControlRigBlueprint /EpicControlRig/Templates/MR_Biped_v02
graphs [
 RigVMFunctionLibrary {links [], nodes []}
 RigVMModel {nodes [RigUnit_BeginExecution = &RigUnit_BeginExecution()]}
]
hierarchy [
 MoveData_Speed {type=Curve, loc=<0 0 0>, rot=<0 0 0 1>}
 base_pose {type=Curve, loc=<0 0 0>, rot=<0 0 0 1>}
 calf_l_back_50 {type=Curve, loc=<0 0 0>, rot=<0 0 0 1>}
 calf_l_back_120 {type=Curve, loc=<0 0 0>, rot=<0 0 0 1>}
 calf_l_back_90 {type=Curve, loc=<0 0 0>, rot=<0 0 0 1>}
 calf_l_back_150 {type=Curve, loc=<0 0 0>, rot=<0 0 0 1>}
 calf_r_back_50 {type=Curve, loc=<0 0 0>, rot=<0 0 0 1>}
 calf_r_back_120 {type=Curve, loc=<0 0 0>, rot=<0 0 0 1>}
 calf_r_back_90 {type=Curve, loc=<0 0 0>, rot=<0 0 0 1>}
 calf_r_back_150 {type=Curve, loc=<0 0 0>, rot=<0 0 0 1>}
 clavicle_l_fwd_30 {type=Curve, loc=<0 0 0>, rot=<0 0 0 1>}
 clavicle_l_down_20 {type=Curve, loc=<0 0 0>, rot=<0 0 0 1>}
 clavicle_l_back_30 {type=Curve, loc=<0 0 0>, rot=<0 0 0 1>}
 clavicle_l_up_40 {type=Curve, loc=<0 0 0>, rot=<0 0 0 1>}
 clavicle_r_fwd_30 {type=Curve, loc=<0 0 0>, rot=<0 0 0 1>}
 clavicle_r_down_20 {type=Curve, loc=<0 0 0>, rot=<0 0 0 1>}
 clavicle_r_back_30 {type=Curve, loc=<0 0 0>, rot=<0 0 0 1>}
 clavicle_r_up_40 {type=Curve, loc=<0 0 0>, rot=<0 0 0 1>}
 clavicle_r_back_310 {type=Curve, loc=<0 0 0>, rot=<0 0 0 1>}
 clavicle_r_up_401 {type=Curve, loc=<0 0 0>, rot=<0 0 0 1>}
 foot_l_down_60 {type=Curve, loc=<0 0 0>, rot=<0 0 0 1>}
 foot_l_up_35 {type=Curve, loc=<0 0 0>, rot=<0 0 0 1>}
 foot_r_up_35 {type=Curve, loc=<0 0 0>, rot=<0 0 0 1>}
 foot_r_down_60 {type=Curve, loc=<0 0 0>, rot=<0 0 0 1>}
 hand_l_down_90 {type=Curve, loc=<0 0 0>, rot=<0 0 0 1>}
 hand_l_up_20 {type=Curve, loc=<0 0 0>, rot=<0 0 0 1>}
 hand_l_up_90 {type=Curve, loc=<0 0 0>, rot=<0 0 0 1>}
 hand_r_up_90 {type=Curve, loc=<0 0 0>, rot=<0 0 0 1>}
 hand_r_down_90 {type=Curve, loc=<0 0 0>, rot=<0 0 0 1>}
 hand_r_up_20 {type=Curve, loc=<0 0 0>, rot=<0 0 0 1>}
 Pose_0 {type=Curve, loc=<0 0 0>, rot=<0 0 0 1>}
 Pose_1 {type=Curve, loc=<0 0 0>, rot=<0 0 0 1>}
 Pose_2 {type=Curve, loc=<0 0 0>, rot=<0 0 0 1>}
 Pose_3 {type=Curve, loc=<0 0 0>, rot=<0 0 0 1>}
 Pose_4 {type=Curve, loc=<0 0 0>, rot=<0 0 0 1>}
 Pose_5 {type=Curve, loc=<0 0 0>, rot=<0 0 0 1>}
 Pose_6 {type=Curve, loc=<0 0 0>, rot=<0 0 0 1>}
 Pose_7 {type=Curve, loc=<0 0 0>, rot=<0 0 0 1>}
 Pose_8 {type=Curve, loc=<0 0 0>, rot=<0 0 0 1>}
 lowerarm_l_in_110 {type=Curve, loc=<0 0 0>, rot=<0 0 0 1>}
 lowerarm_l_in_35 {type=Curve, loc=<0 0 0>, rot=<0 0 0 1>}
 lowerarm_l_out_35 {type=Curve, loc=<0 0 0>, rot=<0 0 0 1>}
 lowerarm_l_in_50 {type=Curve, loc=<0 0 0>, rot=<0 0 0 1>}
 lowerarm_l_in_75 {type=Curve, loc=<0 0 0>, rot=<0 0 0 1>}
 lowerarm_l_in_10 {type=Curve, loc=<0 0 0>, rot=<0 0 0 1>}
 lowerarm_l_out_10 {type=Curve, loc=<0 0 0>, rot=<0 0 0 1>}
 lowerarm_l_in_90 {type=Curve, loc=<0 0 0>, rot=<0 0 0 1>}
 lowerarm_r_in_10 {type=Curve, loc=<0 0 0>, rot=<0 0 0 1>}
 lowerarm_r_out_10 {type=Curve, loc=<0 0 0>, rot=<0 0 0 1>}
 lowerarm_r_out_35 {type=Curve, loc=<0 0 0>, rot=<0 0 0 1>}
 lowerarm_r_in_75 {type=Curve, loc=<0 0 0>, rot=<0 0 0 1>}
 lowerarm_r_in_50 {type=Curve, loc=<0 0 0>, rot=<0 0 0 1>}
 lowerarm_r_in_110 {type=Curve, loc=<0 0 0>, rot=<0 0 0 1>}
 lowerarm_r_in_35 {type=Curve, loc=<0 0 0>, rot=<0 0 0 1>}
 lowerarm_r_in_90 {type=Curve, loc=<0 0 0>, rot=<0 0 0 1>}
 thigh_l_bck_10 {type=Curve, loc=<0 0 0>, rot=<0 0 0 1>}
 thigh_l_fwd_10 {type=Curve, loc=<0 0 0>, rot=<0 0 0 1>}
 thigh_l_in_45_out_90 {type=Curve, loc=<0 0 0>, rot=<0 0 0 1>}
 thigh_l_fwd_90 {type=Curve, loc=<0 0 0>, rot=<0 0 0 1>}
 thigh_l_bck_90 {type=Curve, loc=<0 0 0>, rot=<0 0 0 1>}
 thigh_l_bck_50 {type=Curve, loc=<0 0 0>, rot=<0 0 0 1>}
 thigh_l_out_110 {type=Curve, loc=<0 0 0>, rot=<0 0 0 1>}
 thigh_l_out_10 {type=Curve, loc=<0 0 0>, rot=<0 0 0 1>}
 thigh_l_fwd_45 {type=Curve, loc=<0 0 0>, rot=<0 0 0 1>}
 thigh_l_in_50 {type=Curve, loc=<0 0 0>, rot=<0 0 0 1>}
 thigh_l_fwd_110 {type=Curve, loc=<0 0 0>, rot=<0 0 0 1>}
 thigh_l_out_55 {type=Curve, loc=<0 0 0>, rot=<0 0 0 1>}
 thigh_l_out_85 {type=Curve, loc=<0 0 0>, rot=<0 0 0 1>}
 thigh_r_fwd_10 {type=Curve, loc=<0 0 0>, rot=<0 0 0 1>}
 thigh_r_out_55 {type=Curve, loc=<0 0 0>, rot=<0 0 0 1>}
 thigh_r_fwd_90 {type=Curve, loc=<0 0 0>, rot=<0 0 0 1>}
 thigh_r_fwd_45 {type=Curve, loc=<0 0 0>, rot=<0 0 0 1>}
 thigh_r_bck_10 {type=Curve, loc=<0 0 0>, rot=<0 0 0 1>}
 thigh_r_out_85 {type=Curve, loc=<0 0 0>, rot=<0 0 0 1>}
 thigh_r_in_50 {type=Curve, loc=<0 0 0>, rot=<0 0 0 1>}
 thigh_r_out_10 {type=Curve, loc=<0 0 0>, rot=<0 0 0 1>}
 thigh_r_out_110 {type=Curve, loc=<0 0 0>, rot=<0 0 0 1>}
 thigh_r_bck_50 {type=Curve, loc=<0 0 0>, rot=<0 0 0 1>}
 thigh_r_bck_90 {type=Curve, loc=<0 0 0>, rot=<0 0 0 1>}
 thigh_r_in_45_out_90 {type=Curve, loc=<0 0 0>, rot=<0 0 0 1>}
 thigh_r_fwd_110 {type=Curve, loc=<0 0 0>, rot=<0 0 0 1>}
 upperarm_l_back_45 {type=Curve, loc=<0 0 0>, rot=<0 0 0 1>}
 upperarm_l_out_55 {type=Curve, loc=<0 0 0>, rot=<0 0 0 1>}
 upperarm_l_out110_twist_in_70 {type=Curve, loc=<0 0 0>, rot=<0 0 0 1>}
 upperarm_l_out_85 {type=Curve, loc=<0 0 0>, rot=<0 0 0 1>}
 upperarm_l_fwd_110 {type=Curve, loc=<0 0 0>, rot=<0 0 0 1>}
 upperarm_l_in_10 {type=Curve, loc=<0 0 0>, rot=<0 0 0 1>}
 upperarm_l_out110_twist_out_70 {type=Curve, loc=<0 0 0>, rot=<0 0 0 1>}
 upperarm_l_out_twist_a {type=Curve, loc=<0 0 0>, rot=<0 0 0 1>}
 upperarm_l_fwd_15 {type=Curve, loc=<0 0 0>, rot=<0 0 0 1>}
 upperarm_l_out_15 {type=Curve, loc=<0 0 0>, rot=<0 0 0 1>}
 upperarm_l_back_10 {type=Curve, loc=<0 0 0>, rot=<0 0 0 1>}
 upperarm_l_out110 {type=Curve, loc=<0 0 0>, rot=<0 0 0 1>}
 upperarm_l_fwd_twist_in_a {type=Curve, loc=<0 0 0>, rot=<0 0 0 1>}
 upperarm_l_fwd_90 {type=Curve, loc=<0 0 0>, rot=<0 0 0 1>}
 upperarm_r_out_15 {type=Curve, loc=<0 0 0>, rot=<0 0 0 1>}
 upperarm_r_fwd_15 {type=Curve, loc=<0 0 0>, rot=<0 0 0 1>}
 upperarm_r_out_twist_a {type=Curve, loc=<0 0 0>, rot=<0 0 0 1>}
 upperarm_r_out110_twist_in_70 {type=Curve, loc=<0 0 0>, rot=<0 0 0 1>}
 upperarm_r_fwd_90 {type=Curve, loc=<0 0 0>, rot=<0 0 0 1>}
 upperarm_r_out110 {type=Curve, loc=<0 0 0>, rot=<0 0 0 1>}
 upperarm_r_back_10 {type=Curve, loc=<0 0 0>, rot=<0 0 0 1>}
 upperarm_r_out_85 {type=Curve, loc=<0 0 0>, rot=<0 0 0 1>}
 upperarm_r_in_10 {type=Curve, loc=<0 0 0>, rot=<0 0 0 1>}
 upperarm_r_back_45 {type=Curve, loc=<0 0 0>, rot=<0 0 0 1>}
 upperarm_r_out110_twist_out_70 {type=Curve, loc=<0 0 0>, rot=<0 0 0 1>}
 upperarm_r_out_55 {type=Curve, loc=<0 0 0>, rot=<0 0 0 1>}
 upperarm_r_fwd_twist_in_a {type=Curve, loc=<0 0 0>, rot=<0 0 0 1>}
 upperarm_r_fwd_110 {type=Curve, loc=<0 0 0>, rot=<0 0 0 1>}
 Distance {type=Curve, loc=<0 0 0>, rot=<0 0 0 1>}
 RemainingTurnYaw {type=Curve, loc=<0 0 0>, rot=<0 0 0 1>}
 TurnYawWeight {type=Curve, loc=<0 0 0>, rot=<0 0 0 1>}
 GroundDistance {type=Curve, loc=<0 0 0>, rot=<0 0 0 1>}
 DisableLHandIK {type=Curve, loc=<0 0 0>, rot=<0 0 0 1>}
 ScaleDownWeaponR {type=Curve, loc=<0 0 0>, rot=<0 0 0 1>}
 DisableLegIK {type=Curve, loc=<0 0 0>, rot=<0 0 0 1>}
 blendParent1 {type=Curve, loc=<0 0 0>, rot=<0 0 0 1>}
 DisableRHandIK {type=Curve, loc=<0 0 0>, rot=<0 0 0 1>}
 DisableLeftHandPoseOverride {type=Curve, loc=<0 0 0>, rot=<0 0 0 1>}
 applyHipfireOverridePose {type=Curve, loc=<0 0 0>, rot=<0 0 0 1>}
 DisableHandIKRetargeting {type=Curve, loc=<0 0 0>, rot=<0 0 0 1>}
 lowerarm_r_in_80 {type=Curve, loc=<0 0 0>, rot=<0 0 0 1>}
 lowerarm_l_in_80 {type=Curve, loc=<0 0 0>, rot=<0 0 0 1>}
 calf_r_back_130 {type=Curve, loc=<0 0 0>, rot=<0 0 0 1>}
 calf_l_back_130 {type=Curve, loc=<0 0 0>, rot=<0 0 0 1>}
 Pose_9 {type=Curve, loc=<0 0 0>, rot=<0 0 0 1>}
 BothHandFK {type=Curve, loc=<0 0 0>, rot=<0 0 0 1>}
 TurnRotationAmount {type=Curve, loc=<0 0 0>, rot=<0 0 0 1>}
 DistanceCurve {type=Curve, loc=<0 0 0>, rot=<0 0 0 1>}
 Rocket {type=Curve, loc=<0 0 0>, rot=<0 0 0 1>}
 RootRotation {type=Curve, loc=<0 0 0>, rot=<0 0 0 1>}
 ApplyShoulderPadsAdditive {type=Curve, loc=<0 0 0>, rot=<0 0 0 1>}
 InMotionStart {type=Curve, loc=<0 0 0>, rot=<0 0 0 1>}
 PlayRateSpeedWarpAlpha {type=Curve, loc=<0 0 0>, rot=<0 0 0 1>}
 LookAtPoint {type=Curve, loc=<0 0 0>, rot=<0 0 0 1>}
 lowerarm_l_0_0_n90 {type=Curve, loc=<0 0 0>, rot=<0 0 0 1>}
 lowerarm_r_0_0_n90 {type=Curve, loc=<0 0 0>, rot=<0 0 0 1>}
 l_arm_up {type=Curve, loc=<0 0 0>, rot=<0 0 0 1>}
 r_arm_up {type=Curve, loc=<0 0 0>, rot=<0 0 0 1>}
 l_arm_fwd {type=Curve, loc=<0 0 0>, rot=<0 0 0 1>}
 r_arm_fwd {type=Curve, loc=<0 0 0>, rot=<0 0 0 1>}
 l_arm_bk {type=Curve, loc=<0 0 0>, rot=<0 0 0 1>}
 r_arm_bk {type=Curve, loc=<0 0 0>, rot=<0 0 0 1>}
 l_elbow_bend {type=Curve, loc=<0 0 0>, rot=<0 0 0 1>}
 r_elbow_bend {type=Curve, loc=<0 0 0>, rot=<0 0 0 1>}
 l_wrist_up {type=Curve, loc=<0 0 0>, rot=<0 0 0 1>}
 r_wrist_up {type=Curve, loc=<0 0 0>, rot=<0 0 0 1>}
 l_wrist_dn {type=Curve, loc=<0 0 0>, rot=<0 0 0 1>}
 r_wrist_dn {type=Curve, loc=<0 0 0>, rot=<0 0 0 1>}
 l_knee_bend {type=Curve, loc=<0 0 0>, rot=<0 0 0 1>}
 r_knee_bend {type=Curve, loc=<0 0 0>, rot=<0 0 0 1>}
 r_knee_corrective {type=Curve, loc=<0 0 0>, rot=<0 0 0 1>}
 l_knee_corrective {type=Curve, loc=<0 0 0>, rot=<0 0 0 1>}
 Male_Med_Head_01_Pout {type=Curve, loc=<0 0 0>, rot=<0 0 0 1>}
 Male_Med_Head_02_Pout {type=Curve, loc=<0 0 0>, rot=<0 0 0 1>}
 Male_Med_Head_03_Pout {type=Curve, loc=<0 0 0>, rot=<0 0 0 1>}
 Male_Med_Head_05_Effort {type=Curve, loc=<0 0 0>, rot=<0 0 0 1>}
 Stretch {type=Curve, loc=<0 0 0>, rot=<0 0 0 1>}
 Squash {type=Curve, loc=<0 0 0>, rot=<0 0 0 1>}
 polySurface151 {type=Curve, loc=<0 0 0>, rot=<0 0 0 1>}
 l_cheek_puff {type=Curve, loc=<0 0 0>, rot=<0 0 0 1>}
 r_cheek_puff {type=Curve, loc=<0 0 0>, rot=<0 0 0 1>}
 l_smile_fix {type=Curve, loc=<0 0 0>, rot=<0 0 0 1>}
 r_smile_fix {type=Curve, loc=<0 0 0>, rot=<0 0 0 1>}
 jaw_open_fix {type=Curve, loc=<0 0 0>, rot=<0 0 0 1>}
 r_lip_sneer_fix {type=Curve, loc=<0 0 0>, rot=<0 0 0 1>}
 l_lip_sneer_fix {type=Curve, loc=<0 0 0>, rot=<0 0 0 1>}
 L_Blink {type=Curve, loc=<0 0 0>, rot=<0 0 0 1>}
 R_Blink {type=Curve, loc=<0 0 0>, rot=<0 0 0 1>}
 L_Arm_Stretch {type=Curve, loc=<0 0 0>, rot=<0 0 0 1>}
 Hair_Hat {type=Curve, loc=<0 0 0>, rot=<0 0 0 1>}
 Hair_Helmet {type=Curve, loc=<0 0 0>, rot=<0 0 0 1>}
 Male_Med_Head_02_What {type=Curve, loc=<0 0 0>, rot=<0 0 0 1>}
 Male_Med_Head_02_Smile {type=Curve, loc=<0 0 0>, rot=<0 0 0 1>}
 Male_Med_Head_02_Scowl {type=Curve, loc=<0 0 0>, rot=<0 0 0 1>}
 Male_Med_Head_02_Hmmph {type=Curve, loc=<0 0 0>, rot=<0 0 0 1>}
 Male_Med_Head_02_Grrr {type=Curve, loc=<0 0 0>, rot=<0 0 0 1>}
 Male_Med_Head_02_Effort {type=Curve, loc=<0 0 0>, rot=<0 0 0 1>}
 Male_Med_Head_03_What {type=Curve, loc=<0 0 0>, rot=<0 0 0 1>}
 Male_Med_Head_03_Grrr {type=Curve, loc=<0 0 0>, rot=<0 0 0 1>}
 Male_Med_Head_03_Effort {type=Curve, loc=<0 0 0>, rot=<0 0 0 1>}
 Male_Med_Head_03_Hmmph {type=Curve, loc=<0 0 0>, rot=<0 0 0 1>}
 Male_Med_Head_03_Scowl {type=Curve, loc=<0 0 0>, rot=<0 0 0 1>}
 Male_Med_Head_03_Smile {type=Curve, loc=<0 0 0>, rot=<0 0 0 1>}
 Male_Med_Head_05_Teeth_Low_What {type=Curve, loc=<0 0 0>, rot=<0 0 0 1>}
 Male_Med_Head_05_Teeth_Low_Pout {type=Curve, loc=<0 0 0>, rot=<0 0 0 1>}
 Male_Med_Head_05_Teeth_Low_Hmmph {type=Curve, loc=<0 0 0>, rot=<0 0 0 1>}
 Male_Med_Head_05_Teeth_Low_Grrr {type=Curve, loc=<0 0 0>, rot=<0 0 0 1>}
 Male_Med_Head_05_Teeth_Low_Effort {type=Curve, loc=<0 0 0>, rot=<0 0 0 1>}
 Male_Med_Head_05_What {type=Curve, loc=<0 0 0>, rot=<0 0 0 1>}
 Male_Med_Head_05_Grrr {type=Curve, loc=<0 0 0>, rot=<0 0 0 1>}
 Male_Med_Head_05_Hmmph {type=Curve, loc=<0 0 0>, rot=<0 0 0 1>}
 Male_Med_Head_05_Pout {type=Curve, loc=<0 0 0>, rot=<0 0 0 1>}
 Male_Med_Head_05_Tongue_What {type=Curve, loc=<0 0 0>, rot=<0 0 0 1>}
 Male_Med_Head_05_Tongue_Pout {type=Curve, loc=<0 0 0>, rot=<0 0 0 1>}
 Male_Med_Head_05_Tongue_Hmmph {type=Curve, loc=<0 0 0>, rot=<0 0 0 1>}
 Male_Med_Head_05_Tongue_Grrr {type=Curve, loc=<0 0 0>, rot=<0 0 0 1>}
 Male_Med_Head_05_Tongue_Effort {type=Curve, loc=<0 0 0>, rot=<0 0 0 1>}
 AdditiveLayer {type=Curve, loc=<0 0 0>, rot=<0 0 0 1>}
 RightHandIK {type=Curve, loc=<0 0 0>, rot=<0 0 0 1>}
 LeftHandIK {type=Curve, loc=<0 0 0>, rot=<0 0 0 1>}
 LeftHandFK {type=Curve, loc=<0 0 0>, rot=<0 0 0 1>}
 Male_Med_Head_01_Grrr {type=Curve, loc=<0 0 0>, rot=<0 0 0 1>}
 WeaponAiming {type=Curve, loc=<0 0 0>, rot=<0 0 0 1>}
 Male_Med_Head_01_Hmmph {type=Curve, loc=<0 0 0>, rot=<0 0 0 1>}
 DragonSlash {type=Curve, loc=<0 0 0>, rot=<0 0 0 1>}
 Male_Med_Head_01_Effort {type=Curve, loc=<0 0 0>, rot=<0 0 0 1>}
 LeftHankFK {type=Curve, loc=<0 0 0>, rot=<0 0 0 1>}
 blendOrient1 {type=Curve, loc=<0 0 0>, rot=<0 0 0 1>}
 bottom_R_Corner_Down {type=Curve, loc=<0 0 0>, rot=<0 0 0 1>}
 bottom_R_Corner_Up {type=Curve, loc=<0 0 0>, rot=<0 0 0 1>}
 top_L_Corner_Up {type=Curve, loc=<0 0 0>, rot=<0 0 0 1>}
 top_L_Corner_Down {type=Curve, loc=<0 0 0>, rot=<0 0 0 1>}
 top_R_Corner_Down {type=Curve, loc=<0 0 0>, rot=<0 0 0 1>}
 top_R_Corner_Up {type=Curve, loc=<0 0 0>, rot=<0 0 0 1>}
 bottom_L_Corner_Down {type=Curve, loc=<0 0 0>, rot=<0 0 0 1>}
 bottomt_L_Corner_Up {type=Curve, loc=<0 0 0>, rot=<0 0 0 1>}
 TrailCurve1 {type=Curve, loc=<0 0 0>, rot=<0 0 0 1>}
 bottom_R_Corner_Down1 {type=Curve, loc=<0 0 0>, rot=<0 0 0 1>}
 blendPoint1 {type=Curve, loc=<0 0 0>, rot=<0 0 0 1>}
 WeaponAimingFreeze {type=Curve, loc=<0 0 0>, rot=<0 0 0 1>}
 l_calf_top_taper {type=Curve, loc=<0 0 0>, rot=<0 0 0 1>}
 JogBlendSpaceRotation {type=Curve, loc=<0 0 0>, rot=<0 0 0 1>}
 ZeroOutPitchWeight {type=Curve, loc=<0 0 0>, rot=<0 0 0 1>}
 MeleeTwist {type=Curve, loc=<0 0 0>, rot=<0 0 0 1>}
 MeleeAnim {type=Curve, loc=<0 0 0>, rot=<0 0 0 1>}
 l_sneer_fix {type=Curve, loc=<0 0 0>, rot=<0 0 0 1>}
 jaw_fix {type=Curve, loc=<0 0 0>, rot=<0 0 0 1>}
 r_sneer_fix {type=Curve, loc=<0 0 0>, rot=<0 0 0 1>}
 TurnEnd {type=Curve, loc=<0 0 0>, rot=<0 0 0 1>}
 BlendIn {type=Curve, loc=<0 0 0>, rot=<0 0 0 1>}
 BlendOut {type=Curve, loc=<0 0 0>, rot=<0 0 0 1>}
 BodyRotateX {type=Curve, loc=<0 0 0>, rot=<0 0 0 1>}
 BodyTranslateX {type=Curve, loc=<0 0 0>, rot=<0 0 0 1>}
 CrescentKickScale {type=Curve, loc=<0 0 0>, rot=<0 0 0 1>}
 mask_geo {type=Curve, loc=<0 0 0>, rot=<0 0 0 1>}
 RemoveJumpAdditive {type=Curve, loc=<0 0 0>, rot=<0 0 0 1>}
 RightLeg_IK {type=Curve, loc=<0 0 0>, rot=<0 0 0 1>}
 SnapWeapon_LHand {type=Curve, loc=<0 0 0>, rot=<0 0 0 1>}
 PivotAnim {type=Curve, loc=<0 0 0>, rot=<0 0 0 1>}
 DefaultAdditiveStateWeight {type=Curve, loc=<0 0 0>, rot=<0 0 0 1>}
 StartAnim {type=Curve, loc=<0 0 0>, rot=<0 0 0 1>}
 StopAnim {type=Curve, loc=<0 0 0>, rot=<0 0 0 1>}
 DisableSpeedWarping {type=Curve, loc=<0 0 0>, rot=<0 0 0 1>}
 DisableDBNOSlopeWarp {type=Curve, loc=<0 0 0>, rot=<0 0 0 1>}
 MaskRightArm {type=Curve, loc=<0 0 0>, rot=<0 0 0 1>}
 TurnInPlaceAnim {type=Curve, loc=<0 0 0>, rot=<0 0 0 1>}
 Head_MESH {type=Curve, loc=<0 0 0>, rot=<0 0 0 1>}
 Hair_MESH {type=Curve, loc=<0 0 0>, rot=<0 0 0 1>}
 LaunchPadTransitionOff {type=Curve, loc=<0 0 0>, rot=<0 0 0 1>}
 C_glabella_up_pose {type=Curve, loc=<0 0 0>, rot=<0 0 0 1>}
 jaw_open_pose {type=Curve, loc=<0 0 0>, rot=<0 0 0 1>}
 L_blink_pose {type=Curve, loc=<0 0 0>, rot=<0 0 0 1>}
 L_brow_down_pose {type=Curve, loc=<0 0 0>, rot=<0 0 0 1>}
 L_brow_squeeze_in_pose {type=Curve, loc=<0 0 0>, rot=<0 0 0 1>}
 L_brow_up_pose {type=Curve, loc=<0 0 0>, rot=<0 0 0 1>}
 L_frown_pose {type=Curve, loc=<0 0 0>, rot=<0 0 0 1>}
 L_lip_corner_narrow_pose {type=Curve, loc=<0 0 0>, rot=<0 0 0 1>}
 L_squint_inner_pose {type=Curve, loc=<0 0 0>, rot=<0 0 0 1>}
 L_wide_pose {type=Curve, loc=<0 0 0>, rot=<0 0 0 1>}
 look_down_pose {type=Curve, loc=<0 0 0>, rot=<0 0 0 1>}
 look_left_pose {type=Curve, loc=<0 0 0>, rot=<0 0 0 1>}
 look_right_pose {type=Curve, loc=<0 0 0>, rot=<0 0 0 1>}
 look_up_pose {type=Curve, loc=<0 0 0>, rot=<0 0 0 1>}
 phoneme_fv_pose {type=Curve, loc=<0 0 0>, rot=<0 0 0 1>}
 phoneme_oo_pose {type=Curve, loc=<0 0 0>, rot=<0 0 0 1>}
 R_blink_pose {type=Curve, loc=<0 0 0>, rot=<0 0 0 1>}
 R_brow_down_pose {type=Curve, loc=<0 0 0>, rot=<0 0 0 1>}
 R_brow_squeeze_in_pose {type=Curve, loc=<0 0 0>, rot=<0 0 0 1>}
 R_brow_up_pose {type=Curve, loc=<0 0 0>, rot=<0 0 0 1>}
 R_frown_pose {type=Curve, loc=<0 0 0>, rot=<0 0 0 1>}
 R_lip_corner_narrow_pose {type=Curve, loc=<0 0 0>, rot=<0 0 0 1>}
 R_squint_inner_pose {type=Curve, loc=<0 0 0>, rot=<0 0 0 1>}
 R_wide_pose {type=Curve, loc=<0 0 0>, rot=<0 0 0 1>}
 DisableArmsHeadAdditive {type=Curve, loc=<0 0 0>, rot=<0 0 0 1>}
 FrontendRotation {type=Curve, loc=<0 0 0>, rot=<0 0 0 1>}
 BothLegFK {type=Curve, loc=<0 0 0>, rot=<0 0 0 1>}
 AllowTurnInPlace {type=Curve, loc=<0 0 0>, rot=<0 0 0 1>}
 MontageTimeStretchCurve {type=Curve, loc=<0 0 0>, rot=<0 0 0 1>}
 ParaGlideLeanAlpha {type=Curve, loc=<0 0 0>, rot=<0 0 0 1>}
 WeaponRInChestSpace {type=Curve, loc=<0 0 0>, rot=<0 0 0 1>}
 JimBeam_AO {type=Curve, loc=<0 0 0>, rot=<0 0 0 1>}
 R_squeeze_pose {type=Curve, loc=<0 0 0>, rot=<0 0 0 1>}
 L_squeeze_pose {type=Curve, loc=<0 0 0>, rot=<0 0 0 1>}
 R_upper_lip_raiser_pose {type=Curve, loc=<0 0 0>, rot=<0 0 0 1>}
 L_cheek_up_pose {type=Curve, loc=<0 0 0>, rot=<0 0 0 1>}
 R_cheek_up_pose {type=Curve, loc=<0 0 0>, rot=<0 0 0 1>}
 DisableFaceOverride {type=Curve, loc=<0 0 0>, rot=<0 0 0 1>}
 EnableFullBodyOverride {type=Curve, loc=<0 0 0>, rot=<0 0 0 1>}
 C_glabella_down_pose {type=Curve, loc=<0 0 0>, rot=<0 0 0 1>}
 C_nose_wrinkler_pose {type=Curve, loc=<0 0 0>, rot=<0 0 0 1>}
 eye_convergence_pose {type=Curve, loc=<0 0 0>, rot=<0 0 0 1>}
 jaw_back_pose {type=Curve, loc=<0 0 0>, rot=<0 0 0 1>}
 jaw_forward_pose {type=Curve, loc=<0 0 0>, rot=<0 0 0 1>}
 jaw_left_pose {type=Curve, loc=<0 0 0>, rot=<0 0 0 1>}
 jaw_right_pose {type=Curve, loc=<0 0 0>, rot=<0 0 0 1>}
 jaw_up_pose {type=Curve, loc=<0 0 0>, rot=<0 0 0 1>}
 L_lip_corner_wide_pose {type=Curve, loc=<0 0 0>, rot=<0 0 0 1>}
 L_lower_lip_down_pose {type=Curve, loc=<0 0 0>, rot=<0 0 0 1>}
 L_lower_lip_up_pose {type=Curve, loc=<0 0 0>, rot=<0 0 0 1>}
 L_smile_pose {type=Curve, loc=<0 0 0>, rot=<0 0 0 1>}
 L_upper_lip_lower_pose {type=Curve, loc=<0 0 0>, rot=<0 0 0 1>}
 L_upper_lip_raiser_pose {type=Curve, loc=<0 0 0>, rot=<0 0 0 1>}
 L_upperLidDown_pose {type=Curve, loc=<0 0 0>, rot=<0 0 0 1>}
 mouth_down_pose {type=Curve, loc=<0 0 0>, rot=<0 0 0 1>}
 mouth_up_pose {type=Curve, loc=<0 0 0>, rot=<0 0 0 1>}
 phoneme_ch_pose {type=Curve, loc=<0 0 0>, rot=<0 0 0 1>}
 phoneme_mbp_pose {type=Curve, loc=<0 0 0>, rot=<0 0 0 1>}
 R_lip_corner_wide_pose {type=Curve, loc=<0 0 0>, rot=<0 0 0 1>}
 R_lower_lip_down_pose {type=Curve, loc=<0 0 0>, rot=<0 0 0 1>}
 R_lower_lip_up_pose {type=Curve, loc=<0 0 0>, rot=<0 0 0 1>}
 R_smile_pose {type=Curve, loc=<0 0 0>, rot=<0 0 0 1>}
 R_upper_lip_lower_pose {type=Curve, loc=<0 0 0>, rot=<0 0 0 1>}
 R_upperLidDown_pose {type=Curve, loc=<0 0 0>, rot=<0 0 0 1>}
 tongue_back_pose {type=Curve, loc=<0 0 0>, rot=<0 0 0 1>}
 tongue_down_pose {type=Curve, loc=<0 0 0>, rot=<0 0 0 1>}
 tongue_forward_pose {type=Curve, loc=<0 0 0>, rot=<0 0 0 1>}
 tongue_up_pose {type=Curve, loc=<0 0 0>, rot=<0 0 0 1>}
 LiveLink {type=Curve, loc=<0 0 0>, rot=<0 0 0 1>}
 ParaGlideDeploy {type=Curve, loc=<0 0 0>, rot=<0 0 0 1>}
 HideLeftWeapon {type=Curve, loc=<0 0 0>, rot=<0 0 0 1>}
 DisableLeftHandIKGolfCart {type=Curve, loc=<0 0 0>, rot=<0 0 0 1>}
 DisableRightHandIKGolfCart {type=Curve, loc=<0 0 0>, rot=<0 0 0 1>}
 DisableConsumableOffsetPose {type=Curve, loc=<0 0 0>, rot=<0 0 0 1>}
 UpperBodyNoAdditive {type=Curve, loc=<0 0 0>, rot=<0 0 0 1>}
 MaskLeftArm {type=Curve, loc=<0 0 0>, rot=<0 0 0 1>}
 ForceUpperBody {type=Curve, loc=<0 0 0>, rot=<0 0 0 1>}
 RightHandFK {type=Curve, loc=<0 0 0>, rot=<0 0 0 1>}
 ParaGlideLeanTransition {type=Curve, loc=<0 0 0>, rot=<0 0 0 1>}
 InMotionEmote {type=Curve, loc=<0 0 0>, rot=<0 0 0 1>}
 ForceFullBodyInFrontEnd {type=Curve, loc=<0 0 0>, rot=<0 0 0 1>}
 WeaponLinFloorSpace {type=Curve, loc=<0 0 0>, rot=<0 0 0 1>}
 FrontEnd_Idle_Bypass {type=Curve, loc=<0 0 0>, rot=<0 0 0 1>}
 DisableSlopeWarping {type=Curve, loc=<0 0 0>, rot=<0 0 0 1>}
 Disable_Custom {type=Curve, loc=<0 0 0>, rot=<0 0 0 1>}
 EmoteRequiresWeapon {type=Curve, loc=<0 0 0>, rot=<0 0 0 1>}
 DisableIKRootAdditive {type=Curve, loc=<0 0 0>, rot=<0 0 0 1>}
 DisableFullBodySprint {type=Curve, loc=<0 0 0>, rot=<0 0 0 1>}
 DetachFeet_Vehicle {type=Curve, loc=<0 0 0>, rot=<0 0 0 1>}
 IndigoBoost {type=Curve, loc=<0 0 0>, rot=<0 0 0 1>}
 DisableUpperBodyAdditiveMeshSpace {type=Curve, loc=<0 0 0>, rot=<0 0 0 1>}
 MeleeAOWeight {type=Curve, loc=<0 0 0>, rot=<0 0 0 1>}
 HandsInRootSpace {type=Curve, loc=<0 0 0>, rot=<0 0 0 1>}
 CrouchPelvisAdjustment {type=Curve, loc=<0 0 0>, rot=<0 0 0 1>}
 EnableAccelYawCorrection {type=Curve, loc=<0 0 0>, rot=<0 0 0 1>}
 ParaGlideModifyRoot {type=Curve, loc=<0 0 0>, rot=<0 0 0 1>}
 UseScalarParameterForBlink {type=Curve, loc=<0 0 0>, rot=<0 0 0 1>}
 BlinkControl {type=Curve, loc=<0 0 0>, rot=<0 0 0 1>}
 ForceTargeting {type=Curve, loc=<0 0 0>, rot=<0 0 0 1>}
 BlinkControlProp {type=Curve, loc=<0 0 0>, rot=<0 0 0 1>}
 UseScalarParameterForBlinkProp {type=Curve, loc=<0 0 0>, rot=<0 0 0 1>}
 CrackshotCozyHeadCorrection {type=Curve, loc=<0 0 0>, rot=<0 0 0 1>}
 ForwardMotion {type=Curve, loc=<0 0 0>, rot=<0 0 0 1>}
 EmoteForcedRootYaw {type=Curve, loc=<0 0 0>, rot=<0 0 0 1>}
 MeleeTrailWidth {type=Curve, loc=<0 0 0>, rot=<0 0 0 1>}
 space_world {type=Curve, loc=<0 0 0>, rot=<0 0 0 1>}
 PeelyBlenderFace {type=Curve, loc=<0 0 0>, rot=<0 0 0 1>}
 Width1 {type=Curve, loc=<0 0 0>, rot=<0 0 0 1>}
 OverrideIKHands {type=Curve, loc=<0 0 0>, rot=<0 0 0 1>}
 ParaGlideAddAlpha {type=Curve, loc=<0 0 0>, rot=<0 0 0 1>}
 ParaGlideTurn {type=Curve, loc=<0 0 0>, rot=<0 0 0 1>}
 DisablePelvisAdjustment {type=Curve, loc=<0 0 0>, rot=<0 0 0 1>}
 UncleBrollyDeploy {type=Curve, loc=<0 0 0>, rot=<0 0 0 1>}
 ForceRootSlopeOrientation {type=Curve, loc=<0 0 0>, rot=<0 0 0 1>}
 DisableAdditiveOffsetPose {type=Curve, loc=<0 0 0>, rot=<0 0 0 1>}
 DisableIdleNoiseAdditive {type=Curve, loc=<0 0 0>, rot=<0 0 0 1>}
 UsingConsumable {type=Curve, loc=<0 0 0>, rot=<0 0 0 1>}
 DisableRightArmAdditive {type=Curve, loc=<0 0 0>, rot=<0 0 0 1>}
 NoGlider {type=Curve, loc=<0 0 0>, rot=<0 0 0 1>}
 ParaGlideModifyPelvis {type=Curve, loc=<0 0 0>, rot=<0 0 0 1>}
 ParaGlideModifyLegs {type=Curve, loc=<0 0 0>, rot=<0 0 0 1>}
 HT_L1_TransitionCurve {type=Curve, loc=<0 0 0>, rot=<0 0 0 1>}
 KeepPelvisAdjustment {type=Curve, loc=<0 0 0>, rot=<0 0 0 1>}
 PatrolDisableIdleDuringWalk {type=Curve, loc=<0 0 0>, rot=<0 0 0 1>}
 PatrolWalkUseFullbodyBlendspace {type=Curve, loc=<0 0 0>, rot=<0 0 0 1>}
 RemoveJumpUpAdditive {type=Curve, loc=<0 0 0>, rot=<0 0 0 1>}
 RemoveJumpLoopAdditive {type=Curve, loc=<0 0 0>, rot=<0 0 0 1>}
 RemoveJumpFallAdditive {type=Curve, loc=<0 0 0>, rot=<0 0 0 1>}
 JH_TransformCurve {type=Curve, loc=<0 0 0>, rot=<0 0 0 1>}
 PhysicsOff {type=Curve, loc=<0 0 0>, rot=<0 0 0 1>}
 DisableWeaponAdditivePose {type=Curve, loc=<0 0 0>, rot=<0 0 0 1>}
 Lexa_TransformationCurve {type=Curve, loc=<0 0 0>, rot=<0 0 0 1>}
 HideProp {type=Curve, loc=<0 0 0>, rot=<0 0 0 1>}
 CopyPose {type=Curve, loc=<0 0 0>, rot=<0 0 0 1>}
 MaterialCurveA {type=Curve, loc=<0 0 0>, rot=<0 0 0 1>}
 MaterialCurveB {type=Curve, loc=<0 0 0>, rot=<0 0 0 1>}
 MaterialCurveC {type=Curve, loc=<0 0 0>, rot=<0 0 0 1>}
 MaterialCurveD {type=Curve, loc=<0 0 0>, rot=<0 0 0 1>}
 MaterialCurveE {type=Curve, loc=<0 0 0>, rot=<0 0 0 1>}
 UseEmoteAnimOffset {type=Curve, loc=<0 0 0>, rot=<0 0 0 1>}
 emissive mult 2 {type=Curve, loc=<0 0 0>, rot=<0 0 0 1>}
 MaxHandle {type=Curve, loc=<0 0 0>, rot=<0 0 0 1>}
 FootPhase {type=Curve, loc=<0 0 0>, rot=<0 0 0 1>}
 TacticalDoorBash {type=Curve, loc=<0 0 0>, rot=<0 0 0 1>}
 Custom_Frontend_Face_Idle {type=Curve, loc=<0 0 0>, rot=<0 0 0 1>}
 HandIKInAttachBoneSpace {type=Curve, loc=<0 0 0>, rot=<0 0 0 1>}
 SwingPhase {type=Curve, loc=<0 0 0>, rot=<0 0 0 1>}
 Moving {type=Curve, loc=<0 0 0>, rot=<0 0 0 1>}
 HandAttachRInChestSpace {type=Curve, loc=<0 0 0>, rot=<0 0 0 1>}
 emote_BuiltIn {type=Curve, loc=<0 0 0>, rot=<0 0 0 1>}
 OhanaPackAdj {type=Curve, loc=<0 0 0>, rot=<0 0 0 1>}
 UseRootSlopeOrientation {type=Curve, loc=<0 0 0>, rot=<0 0 0 1>}
 Backpack_Anim_Override {type=Curve, loc=<0 0 0>, rot=<0 0 0 1>}
 PhysicsGroundOffset {type=Curve, loc=<0 0 0>, rot=<0 0 0 1>}
 AllowRidingNoiseAdditive {type=Curve, loc=<0 0 0>, rot=<0 0 0 1>}
 LeftLegIKOnGroundVehicle {type=Curve, loc=<0 0 0>, rot=<0 0 0 1>}
 RightLegIKOnGroundVehicle {type=Curve, loc=<0 0 0>, rot=<0 0 0 1>}
 BaseHeadDefaultFaceIdle {type=Curve, loc=<0 0 0>, rot=<0 0 0 1>}
 Hurdle_100fwd {type=Curve, loc=<0 0 0>, rot=<0 0 0 1>}
 Hurdle_250fwd {type=Curve, loc=<0 0 0>, rot=<0 0 0 1>}
 Hurdle_100up {type=Curve, loc=<0 0 0>, rot=<0 0 0 1>}
 Hurdle_200up {type=Curve, loc=<0 0 0>, rot=<0 0 0 1>}
 Hurdle_50up {type=Curve, loc=<0 0 0>, rot=<0 0 0 1>}
 Hurdle_Mantle {type=Curve, loc=<0 0 0>, rot=<0 0 0 1>}
 Idle {type=Curve, loc=<0 0 0>, rot=<0 0 0 1>}
 Interruptible {type=Curve, loc=<0 0 0>, rot=<0 0 0 1>}
 CustomFaceIdle {type=Curve, loc=<0 0 0>, rot=<0 0 0 1>}
 Hurdle_Land {type=Curve, loc=<0 0 0>, rot=<0 0 0 1>}
 Hurdle_Fall {type=Curve, loc=<0 0 0>, rot=<0 0 0 1>}
 TacticalJump {type=Curve, loc=<0 0 0>, rot=<0 0 0 1>}
 copyWeaponRtoAttachL {type=Curve, loc=<0 0 0>, rot=<0 0 0 1>}
 CTRL_expressions_browDownL {type=Curve, loc=<0 0 0>, rot=<0 0 0 1>}
 CTRL_expressions_browDownR {type=Curve, loc=<0 0 0>, rot=<0 0 0 1>}
 CTRL_expressions_browLateralL {type=Curve, loc=<0 0 0>, rot=<0 0 0 1>}
 CTRL_expressions_browLateralR {type=Curve, loc=<0 0 0>, rot=<0 0 0 1>}
 CTRL_expressions_browRaiseInL {type=Curve, loc=<0 0 0>, rot=<0 0 0 1>}
 CTRL_expressions_browRaiseInR {type=Curve, loc=<0 0 0>, rot=<0 0 0 1>}
 CTRL_expressions_browRaiseOuterL {type=Curve, loc=<0 0 0>, rot=<0 0 0 1>}
 CTRL_expressions_browRaiseOuterR {type=Curve, loc=<0 0 0>, rot=<0 0 0 1>}
 CTRL_expressions_browUINL {type=Curve, loc=<0 0 0>, rot=<0 0 0 1>}
 CTRL_expressions_browUINR {type=Curve, loc=<0 0 0>, rot=<0 0 0 1>}
 CTRL_expressions_browUOUTL {type=Curve, loc=<0 0 0>, rot=<0 0 0 1>}
 CTRL_expressions_browUOUTR {type=Curve, loc=<0 0 0>, rot=<0 0 0 1>}
 CTRL_expressions_browDINL {type=Curve, loc=<0 0 0>, rot=<0 0 0 1>}
 CTRL_expressions_browDINR {type=Curve, loc=<0 0 0>, rot=<0 0 0 1>}
 CTRL_expressions_browDOUTL {type=Curve, loc=<0 0 0>, rot=<0 0 0 1>}
 CTRL_expressions_browDOUTR {type=Curve, loc=<0 0 0>, rot=<0 0 0 1>}
 CTRL_expressions_eyeBlinkL {type=Curve, loc=<0 0 0>, rot=<0 0 0 1>}
 CTRL_expressions_eyeBlinkR {type=Curve, loc=<0 0 0>, rot=<0 0 0 1>}
 CTRL_expressions_eyeWidenL {type=Curve, loc=<0 0 0>, rot=<0 0 0 1>}
 CTRL_expressions_eyeWidenR {type=Curve, loc=<0 0 0>, rot=<0 0 0 1>}
 CTRL_expressions_eyeSquintInnerL {type=Curve, loc=<0 0 0>, rot=<0 0 0 1>}
 CTRL_expressions_eyeSquintInnerR {type=Curve, loc=<0 0 0>, rot=<0 0 0 1>}
 CTRL_expressions_eyeCheekRaiseL {type=Curve, loc=<0 0 0>, rot=<0 0 0 1>}
 CTRL_expressions_eyeCheekRaiseR {type=Curve, loc=<0 0 0>, rot=<0 0 0 1>}
 CTRL_expressions_eyeUpperLidUpL {type=Curve, loc=<0 0 0>, rot=<0 0 0 1>}
 CTRL_expressions_eyeUpperLidUpR {type=Curve, loc=<0 0 0>, rot=<0 0 0 1>}
 CTRL_expressions_eyeRelaxL {type=Curve, loc=<0 0 0>, rot=<0 0 0 1>}
 CTRL_expressions_eyeRelaxR {type=Curve, loc=<0 0 0>, rot=<0 0 0 1>}
 CTRL_expressions_eyeLowerLidUpL {type=Curve, loc=<0 0 0>, rot=<0 0 0 1>}
 CTRL_expressions_eyeLowerLidUpR {type=Curve, loc=<0 0 0>, rot=<0 0 0 1>}
 CTRL_expressions_eyeLowerLidDownL {type=Curve, loc=<0 0 0>, rot=<0 0 0 1>}
 CTRL_expressions_eyeLowerLidDownR {type=Curve, loc=<0 0 0>, rot=<0 0 0 1>}
 CTRL_expressions_eyelidRollPosUL {type=Curve, loc=<0 0 0>, rot=<0 0 0 1>}
 CTRL_expressions_eyelidRollPosUR {type=Curve, loc=<0 0 0>, rot=<0 0 0 1>}
 CTRL_expressions_eyelidRollPosDL {type=Curve, loc=<0 0 0>, rot=<0 0 0 1>}
 CTRL_expressions_eyelidRollPosDR {type=Curve, loc=<0 0 0>, rot=<0 0 0 1>}
 CTRL_expressions_eyelidRollNegUL {type=Curve, loc=<0 0 0>, rot=<0 0 0 1>}
 CTRL_expressions_eyelidRollNegUR {type=Curve, loc=<0 0 0>, rot=<0 0 0 1>}
 CTRL_expressions_eyelidRollNegDL {type=Curve, loc=<0 0 0>, rot=<0 0 0 1>}
 CTRL_expressions_eyelidRollNegDR {type=Curve, loc=<0 0 0>, rot=<0 0 0 1>}
 CTRL_expressions_eyeLookUpL {type=Curve, loc=<0 0 0>, rot=<0 0 0 1>}
 CTRL_expressions_eyeLookUpR {type=Curve, loc=<0 0 0>, rot=<0 0 0 1>}
 CTRL_expressions_eyeLookDownL {type=Curve, loc=<0 0 0>, rot=<0 0 0 1>}
 CTRL_expressions_eyeLookDownR {type=Curve, loc=<0 0 0>, rot=<0 0 0 1>}
 CTRL_expressions_eyeLookLeftL {type=Curve, loc=<0 0 0>, rot=<0 0 0 1>}
 CTRL_expressions_eyeLookLeftR {type=Curve, loc=<0 0 0>, rot=<0 0 0 1>}
 CTRL_expressions_eyeLookRightL {type=Curve, loc=<0 0 0>, rot=<0 0 0 1>}
 CTRL_expressions_eyeLookRightR {type=Curve, loc=<0 0 0>, rot=<0 0 0 1>}
 CTRL_expressions_eyeEyelidsReduce {type=Curve, loc=<0 0 0>, rot=<0 0 0 1>}
 CTRL_expressions_noseWrinkleL {type=Curve, loc=<0 0 0>, rot=<0 0 0 1>}
 CTRL_expressions_noseWrinkleR {type=Curve, loc=<0 0 0>, rot=<0 0 0 1>}
 CTRL_expressions_noseWrinkleUpperL {type=Curve, loc=<0 0 0>, rot=<0 0 0 1>}
 CTRL_expressions_noseWrinkleUpperR {type=Curve, loc=<0 0 0>, rot=<0 0 0 1>}
 CTRL_expressions_noseNostrilDepressL {type=Curve, loc=<0 0 0>, rot=<0 0 0 1>}
 CTRL_expressions_noseNostrilDepressR {type=Curve, loc=<0 0 0>, rot=<0 0 0 1>}
 CTRL_expressions_noseNostrilDilateL {type=Curve, loc=<0 0 0>, rot=<0 0 0 1>}
 CTRL_expressions_noseNostrilDilateR {type=Curve, loc=<0 0 0>, rot=<0 0 0 1>}
 CTRL_expressions_noseNostrilCompressL {type=Curve, loc=<0 0 0>, rot=<0 0 0 1>}
 CTRL_expressions_noseNostrilCompressR {type=Curve, loc=<0 0 0>, rot=<0 0 0 1>}
 CTRL_expressions_noseRotateL {type=Curve, loc=<0 0 0>, rot=<0 0 0 1>}
 CTRL_expressions_noseRotateR {type=Curve, loc=<0 0 0>, rot=<0 0 0 1>}
 CTRL_expressions_noseSquash {type=Curve, loc=<0 0 0>, rot=<0 0 0 1>}
 CTRL_expressions_noseStretch {type=Curve, loc=<0 0 0>, rot=<0 0 0 1>}
 CTRL_expressions_mouthCheekSuckL {type=Curve, loc=<0 0 0>, rot=<0 0 0 1>}
 CTRL_expressions_mouthCheekSuckR {type=Curve, loc=<0 0 0>, rot=<0 0 0 1>}
 CTRL_expressions_mouthCheekBlowL {type=Curve, loc=<0 0 0>, rot=<0 0 0 1>}
 CTRL_expressions_mouthCheekBlowR {type=Curve, loc=<0 0 0>, rot=<0 0 0 1>}
 CTRL_expressions_mouthLipsBlowL {type=Curve, loc=<0 0 0>, rot=<0 0 0 1>}
 CTRL_expressions_mouthLipsBlowR {type=Curve, loc=<0 0 0>, rot=<0 0 0 1>}
 CTRL_expressions_mouthLeft {type=Curve, loc=<0 0 0>, rot=<0 0 0 1>}
 CTRL_expressions_mouthRight {type=Curve, loc=<0 0 0>, rot=<0 0 0 1>}
 CTRL_expressions_mouthUp {type=Curve, loc=<0 0 0>, rot=<0 0 0 1>}
 CTRL_expressions_mouthDown {type=Curve, loc=<0 0 0>, rot=<0 0 0 1>}
 CTRL_expressions_mouthUpperLipRaiseL {type=Curve, loc=<0 0 0>, rot=<0 0 0 1>}
 CTRL_expressions_mouthUpperLipRaiseR {type=Curve, loc=<0 0 0>, rot=<0 0 0 1>}
 CTRL_expressions_mouthLowerLipDepressL {type=Curve, loc=<0 0 0>, rot=<0 0 0 1>}
 CTRL_expressions_mouthLowerLipDepressR {type=Curve, loc=<0 0 0>, rot=<0 0 0 1>}
 CTRL_expressions_mouthCornerPullL {type=Curve, loc=<0 0 0>, rot=<0 0 0 1>}
 CTRL_expressions_mouthCornerPullR {type=Curve, loc=<0 0 0>, rot=<0 0 0 1>}
 CTRL_expressions_mouthStretchL {type=Curve, loc=<0 0 0>, rot=<0 0 0 1>}
 CTRL_expressions_mouthStretchR {type=Curve, loc=<0 0 0>, rot=<0 0 0 1>}
 CTRL_expressions_mouthDimpleL {type=Curve, loc=<0 0 0>, rot=<0 0 0 1>}
 CTRL_expressions_mouthDimpleR {type=Curve, loc=<0 0 0>, rot=<0 0 0 1>}
 CTRL_expressions_mouthCornerDepressL {type=Curve, loc=<0 0 0>, rot=<0 0 0 1>}
 CTRL_expressions_mouthCornerDepressR {type=Curve, loc=<0 0 0>, rot=<0 0 0 1>}
 CTRL_expressions_mouthPressUL {type=Curve, loc=<0 0 0>, rot=<0 0 0 1>}
 CTRL_expressions_mouthPressUR {type=Curve, loc=<0 0 0>, rot=<0 0 0 1>}
 CTRL_expressions_mouthPressDL {type=Curve, loc=<0 0 0>, rot=<0 0 0 1>}
 CTRL_expressions_mouthPressDR {type=Curve, loc=<0 0 0>, rot=<0 0 0 1>}
 CTRL_expressions_mouthLipsPurseUL {type=Curve, loc=<0 0 0>, rot=<0 0 0 1>}
 CTRL_expressions_mouthLipsPurseUR {type=Curve, loc=<0 0 0>, rot=<0 0 0 1>}
 CTRL_expressions_mouthLipsPurseDL {type=Curve, loc=<0 0 0>, rot=<0 0 0 1>}
 CTRL_expressions_mouthLipsPurseDR {type=Curve, loc=<0 0 0>, rot=<0 0 0 1>}
 CTRL_expressions_mouthLipsTowardsUL {type=Curve, loc=<0 0 0>, rot=<0 0 0 1>}
 CTRL_expressions_mouthLipsTowardsUR {type=Curve, loc=<0 0 0>, rot=<0 0 0 1>}
 CTRL_expressions_mouthLipsTowardsDL {type=Curve, loc=<0 0 0>, rot=<0 0 0 1>}
 CTRL_expressions_mouthLipsTowardsDR {type=Curve, loc=<0 0 0>, rot=<0 0 0 1>}
 CTRL_expressions_mouthFunnelUL {type=Curve, loc=<0 0 0>, rot=<0 0 0 1>}
 CTRL_expressions_mouthFunnelUR {type=Curve, loc=<0 0 0>, rot=<0 0 0 1>}
 CTRL_expressions_mouthFunnelDL {type=Curve, loc=<0 0 0>, rot=<0 0 0 1>}
 CTRL_expressions_mouthFunnelDR {type=Curve, loc=<0 0 0>, rot=<0 0 0 1>}
 CTRL_expressions_mouthLipsTogetherUL {type=Curve, loc=<0 0 0>, rot=<0 0 0 1>}
 CTRL_expressions_mouthLipsTogetherUR {type=Curve, loc=<0 0 0>, rot=<0 0 0 1>}
 CTRL_expressions_mouthLipsTogetherDL {type=Curve, loc=<0 0 0>, rot=<0 0 0 1>}
 CTRL_expressions_mouthLipsTogetherDR {type=Curve, loc=<0 0 0>, rot=<0 0 0 1>}
 CTRL_expressions_mouthUpperLipBiteL {type=Curve, loc=<0 0 0>, rot=<0 0 0 1>}
 CTRL_expressions_mouthUpperLipBiteR {type=Curve, loc=<0 0 0>, rot=<0 0 0 1>}
 CTRL_expressions_mouthLowerLipBiteL {type=Curve, loc=<0 0 0>, rot=<0 0 0 1>}
 CTRL_expressions_mouthLowerLipBiteR {type=Curve, loc=<0 0 0>, rot=<0 0 0 1>}
 CTRL_expressions_mouthCornerSharpenUL {type=Curve, loc=<0 0 0>, rot=<0 0 0 1>}
 CTRL_expressions_mouthCornerSharpenUR {type=Curve, loc=<0 0 0>, rot=<0 0 0 1>}
 CTRL_expressions_mouthCornerSharpenDL {type=Curve, loc=<0 0 0>, rot=<0 0 0 1>}
 CTRL_expressions_mouthCornerSharpenDR {type=Curve, loc=<0 0 0>, rot=<0 0 0 1>}
 CTRL_expressions_mouthCornerRounderUL {type=Curve, loc=<0 0 0>, rot=<0 0 0 1>}
 CTRL_expressions_mouthCornerRounderUR {type=Curve, loc=<0 0 0>, rot=<0 0 0 1>}
 CTRL_expressions_mouthCornerRounderDL {type=Curve, loc=<0 0 0>, rot=<0 0 0 1>}
 CTRL_expressions_mouthCornerRounderDR {type=Curve, loc=<0 0 0>, rot=<0 0 0 1>}
 CTRL_expressions_mouthUpperLipTowardsTeethL {type=Curve, loc=<0 0 0>, rot=<0 0 0 1>}
 CTRL_expressions_mouthUpperLipTowardsTeethR {type=Curve, loc=<0 0 0>, rot=<0 0 0 1>}
 CTRL_expressions_mouthLowerLipTowardsTeethL {type=Curve, loc=<0 0 0>, rot=<0 0 0 1>}
 CTRL_expressions_mouthLowerLipTowardsTeethR {type=Curve, loc=<0 0 0>, rot=<0 0 0 1>}
 CTRL_expressions_mouthCornerUpL {type=Curve, loc=<0 0 0>, rot=<0 0 0 1>}
 CTRL_expressions_mouthCornerUpR {type=Curve, loc=<0 0 0>, rot=<0 0 0 1>}
 CTRL_expressions_mouthCornerDownL {type=Curve, loc=<0 0 0>, rot=<0 0 0 1>}
 CTRL_expressions_mouthCornerDownR {type=Curve, loc=<0 0 0>, rot=<0 0 0 1>}
 CTRL_expressions_mouthCornerWideL {type=Curve, loc=<0 0 0>, rot=<0 0 0 1>}
 CTRL_expressions_mouthCornerWideR {type=Curve, loc=<0 0 0>, rot=<0 0 0 1>}
 CTRL_expressions_mouthCornerNarrowL {type=Curve, loc=<0 0 0>, rot=<0 0 0 1>}
 CTRL_expressions_mouthCornerNarrowR {type=Curve, loc=<0 0 0>, rot=<0 0 0 1>}
 CTRL_expressions_tongueUp {type=Curve, loc=<0 0 0>, rot=<0 0 0 1>}
 CTRL_expressions_tongueDown {type=Curve, loc=<0 0 0>, rot=<0 0 0 1>}
 CTRL_expressions_tongueLeft {type=Curve, loc=<0 0 0>, rot=<0 0 0 1>}
 CTRL_expressions_tongueRight {type=Curve, loc=<0 0 0>, rot=<0 0 0 1>}
 CTRL_expressions_tongueOut {type=Curve, loc=<0 0 0>, rot=<0 0 0 1>}
 CTRL_expressions_tongueIn {type=Curve, loc=<0 0 0>, rot=<0 0 0 1>}
 CTRL_expressions_tongueRollUp {type=Curve, loc=<0 0 0>, rot=<0 0 0 1>}
 CTRL_expressions_tongueRollDown {type=Curve, loc=<0 0 0>, rot=<0 0 0 1>}
 CTRL_expressions_tongueRollLeft {type=Curve, loc=<0 0 0>, rot=<0 0 0 1>}
 CTRL_expressions_tongueRollRight {type=Curve, loc=<0 0 0>, rot=<0 0 0 1>}
 CTRL_expressions_tongueWide {type=Curve, loc=<0 0 0>, rot=<0 0 0 1>}
 CTRL_expressions_tongueNarrow {type=Curve, loc=<0 0 0>, rot=<0 0 0 1>}
 CTRL_expressions_jawOpen {type=Curve, loc=<0 0 0>, rot=<0 0 0 1>}
 CTRL_expressions_jawLeft {type=Curve, loc=<0 0 0>, rot=<0 0 0 1>}
 CTRL_expressions_jawRight {type=Curve, loc=<0 0 0>, rot=<0 0 0 1>}
 CTRL_expressions_jawFwd {type=Curve, loc=<0 0 0>, rot=<0 0 0 1>}
 CTRL_expressions_jawBack {type=Curve, loc=<0 0 0>, rot=<0 0 0 1>}
 CTRL_expressions_jawChinRaiseDL {type=Curve, loc=<0 0 0>, rot=<0 0 0 1>}
 CTRL_expressions_jawChinRaiseDR {type=Curve, loc=<0 0 0>, rot=<0 0 0 1>}
 CTRL_expressions_jawChinRaiseUL {type=Curve, loc=<0 0 0>, rot=<0 0 0 1>}
 CTRL_expressions_jawChinRaiseUR {type=Curve, loc=<0 0 0>, rot=<0 0 0 1>}
 CTRL_expressions_jawOpenExtreme {type=Curve, loc=<0 0 0>, rot=<0 0 0 1>}
 CTRL_expressions_teethUpU {type=Curve, loc=<0 0 0>, rot=<0 0 0 1>}
 CTRL_expressions_teethUpD {type=Curve, loc=<0 0 0>, rot=<0 0 0 1>}
 CTRL_expressions_teethDownU {type=Curve, loc=<0 0 0>, rot=<0 0 0 1>}
 CTRL_expressions_teethDownD {type=Curve, loc=<0 0 0>, rot=<0 0 0 1>}
 CTRL_expressions_teethLeftU {type=Curve, loc=<0 0 0>, rot=<0 0 0 1>}
 CTRL_expressions_teethLeftD {type=Curve, loc=<0 0 0>, rot=<0 0 0 1>}
 CTRL_expressions_teethRightU {type=Curve, loc=<0 0 0>, rot=<0 0 0 1>}
 CTRL_expressions_teethRightD {type=Curve, loc=<0 0 0>, rot=<0 0 0 1>}
 CTRL_expressions_teethFwdU {type=Curve, loc=<0 0 0>, rot=<0 0 0 1>}
 CTRL_expressions_teethFwdD {type=Curve, loc=<0 0 0>, rot=<0 0 0 1>}
 CTRL_expressions_teethBackU {type=Curve, loc=<0 0 0>, rot=<0 0 0 1>}
 CTRL_expressions_teethBackD {type=Curve, loc=<0 0 0>, rot=<0 0 0 1>}
 CTRL_expressions_teethRotateFwdU {type=Curve, loc=<0 0 0>, rot=<0 0 0 1>}
 CTRL_expressions_teethRotateFwdD {type=Curve, loc=<0 0 0>, rot=<0 0 0 1>}
 CTRL_expressions_teethRotateBackU {type=Curve, loc=<0 0 0>, rot=<0 0 0 1>}
 CTRL_expressions_teethRotateBackD {type=Curve, loc=<0 0 0>, rot=<0 0 0 1>}
 is_3L {type=Curve, loc=<0 0 0>, rot=<0 0 0 1>}
 Mirrored {type=Curve, loc=<0 0 0>, rot=<0 0 0 1>}
 LockMovement {type=Curve, loc=<0 0 0>, rot=<0 0 0 1>}
 Walk {type=Curve, loc=<0 0 0>, rot=<0 0 0 1>}
 Sprint {type=Curve, loc=<0 0 0>, rot=<0 0 0 1>}
 Jog {type=Curve, loc=<0 0 0>, rot=<0 0 0 1>}
 Run {type=Curve, loc=<0 0 0>, rot=<0 0 0 1>}
 LeftFootOnGrindRail {type=Curve, loc=<0 0 0>, rot=<0 0 0 1>}
 RightFootOnGrindRail {type=Curve, loc=<0 0 0>, rot=<0 0 0 1>}
 DisableLeftHandIKHipAttach {type=Curve, loc=<0 0 0>, rot=<0 0 0 1>}
 YawCorrectionDirectionSouth {type=Curve, loc=<0 0 0>, rot=<0 0 0 1>}
 YawCorrectionDirectionEast {type=Curve, loc=<0 0 0>, rot=<0 0 0 1>}
 YawCorrectionDirectionWest {type=Curve, loc=<0 0 0>, rot=<0 0 0 1>}
 YawCorrectionDirectionNorth {type=Curve, loc=<0 0 0>, rot=<0 0 0 1>}
 is_face_bone_anim {type=Curve, loc=<0 0 0>, rot=<0 0 0 1>}
 Hover {type=Curve, loc=<0 0 0>, rot=<0 0 0 1>}
 WeaponRinFloorSpace {type=Curve, loc=<0 0 0>, rot=<0 0 0 1>}
 AlphaBlur_R {type=Curve, loc=<0 0 0>, rot=<0 0 0 1>}
 AlphaBlur_L {type=Curve, loc=<0 0 0>, rot=<0 0 0 1>}
 RotationDirection_R {type=Curve, loc=<0 0 0>, rot=<0 0 0 1>}
 RotationDirection_L {type=Curve, loc=<0 0 0>, rot=<0 0 0 1>}
 PlayLowerBodyLocomotionOverFullBodySlot {type=Curve, loc=<0 0 0>, rot=<0 0 0 1>}
 UserCurve01 {type=Curve, loc=<0 0 0>, rot=<0 0 0 1>}
 UserCurve02 {type=Curve, loc=<0 0 0>, rot=<0 0 0 1>}
 LegIKInAttachBoneSpace {type=Curve, loc=<0 0 0>, rot=<0 0 0 1>}
 FootSpeed_L {type=Curve, loc=<0 0 0>, rot=<0 0 0 1>}
 FootSpeed_R {type=Curve, loc=<0 0 0>, rot=<0 0 0 1>}
 RightElbowTwist {type=Curve, loc=<0 0 0>, rot=<0 0 0 1>}
 LeftElbowTwist {type=Curve, loc=<0 0 0>, rot=<0 0 0 1>}
 Intensity {type=Curve, loc=<0 0 0>, rot=<0 0 0 1>}
 head {type=Curve, loc=<0 0 0>, rot=<0 0 0 1>}
 neck_01 {type=Curve, loc=<0 0 0>, rot=<0 0 0 1>}
 neck_02 {type=Curve, loc=<0 0 0>, rot=<0 0 0 1>}
 spine_01 {type=Curve, loc=<0 0 0>, rot=<0 0 0 1>}
 spine_04 {type=Curve, loc=<0 0 0>, rot=<0 0 0 1>}
 spine_03 {type=Curve, loc=<0 0 0>, rot=<0 0 0 1>}
 spine_05 {type=Curve, loc=<0 0 0>, rot=<0 0 0 1>}
 upperarm_r {type=Curve, loc=<0 0 0>, rot=<0 0 0 1>}
 hand_l {type=Curve, loc=<0 0 0>, rot=<0 0 0 1>}
 lowerarm_l {type=Curve, loc=<0 0 0>, rot=<0 0 0 1>}
 lowerarm_r {type=Curve, loc=<0 0 0>, rot=<0 0 0 1>}
 lowerarm_twist_01_l {type=Curve, loc=<0 0 0>, rot=<0 0 0 1>}
 upperarm_l {type=Curve, loc=<0 0 0>, rot=<0 0 0 1>}
 ball_r {type=Curve, loc=<0 0 0>, rot=<0 0 0 1>}
 index_metacarpal_l {type=Curve, loc=<0 0 0>, rot=<0 0 0 1>}
 clavicle_l {type=Curve, loc=<0 0 0>, rot=<0 0 0 1>}
 hand_r {type=Curve, loc=<0 0 0>, rot=<0 0 0 1>}
 spine_02 {type=Curve, loc=<0 0 0>, rot=<0 0 0 1>}
 calf_r {type=Curve, loc=<0 0 0>, rot=<0 0 0 1>}
 calf_twist_01_r {type=Curve, loc=<0 0 0>, rot=<0 0 0 1>}
 calf_twist_02_r {type=Curve, loc=<0 0 0>, rot=<0 0 0 1>}
 clavicle_r {type=Curve, loc=<0 0 0>, rot=<0 0 0 1>}
 ik_hand_r {type=Curve, loc=<0 0 0>, rot=<0 0 0 1>}
 thigh_r {type=Curve, loc=<0 0 0>, rot=<0 0 0 1>}
 thigh_twist_01_r {type=Curve, loc=<0 0 0>, rot=<0 0 0 1>}
 calf_l {type=Curve, loc=<0 0 0>, rot=<0 0 0 1>}
 thigh_l {type=Curve, loc=<0 0 0>, rot=<0 0 0 1>}
 pelvis {type=Curve, loc=<0 0 0>, rot=<0 0 0 1>}
 thumb_01_l {type=Curve, loc=<0 0 0>, rot=<0 0 0 1>}
 thumb_02_l {type=Curve, loc=<0 0 0>, rot=<0 0 0 1>}
 thumb_03_l {type=Curve, loc=<0 0 0>, rot=<0 0 0 1>}
 middle_metacarpal_l {type=Curve, loc=<0 0 0>, rot=<0 0 0 1>}
 middle_01_l {type=Curve, loc=<0 0 0>, rot=<0 0 0 1>}
 middle_02_l {type=Curve, loc=<0 0 0>, rot=<0 0 0 1>}
 middle_03_l {type=Curve, loc=<0 0 0>, rot=<0 0 0 1>}
 pinky_01_l {type=Curve, loc=<0 0 0>, rot=<0 0 0 1>}
 pinky_02_l {type=Curve, loc=<0 0 0>, rot=<0 0 0 1>}
 pinky_03_l {type=Curve, loc=<0 0 0>, rot=<0 0 0 1>}
 ring_01_l {type=Curve, loc=<0 0 0>, rot=<0 0 0 1>}
 ring_02_l {type=Curve, loc=<0 0 0>, rot=<0 0 0 1>}
 ring_03_l {type=Curve, loc=<0 0 0>, rot=<0 0 0 1>}
 index_01_r {type=Curve, loc=<0 0 0>, rot=<0 0 0 1>}
 index_02_r {type=Curve, loc=<0 0 0>, rot=<0 0 0 1>}
 index_03_r {type=Curve, loc=<0 0 0>, rot=<0 0 0 1>}
 lowerarm_twist_01_r {type=Curve, loc=<0 0 0>, rot=<0 0 0 1>}
 lowerarm_twist_02_r {type=Curve, loc=<0 0 0>, rot=<0 0 0 1>}
 pinky_metacarpal_r {type=Curve, loc=<0 0 0>, rot=<0 0 0 1>}
 ring_01_r {type=Curve, loc=<0 0 0>, rot=<0 0 0 1>}
 ring_02_r {type=Curve, loc=<0 0 0>, rot=<0 0 0 1>}
 thumb_01_r {type=Curve, loc=<0 0 0>, rot=<0 0 0 1>}
 thumb_02_r {type=Curve, loc=<0 0 0>, rot=<0 0 0 1>}
 thumb_03_r {type=Curve, loc=<0 0 0>, rot=<0 0 0 1>}
 foot_r {type=Curve, loc=<0 0 0>, rot=<0 0 0 1>}
 upperarm_twist_01_l {type=Curve, loc=<0 0 0>, rot=<0 0 0 1>}
 upperarm_twist_01_r {type=Curve, loc=<0 0 0>, rot=<0 0 0 1>}
 root {type=Curve, loc=<0 0 0>, rot=<0 0 0 1>}
 foot_l {type=Curve, loc=<0 0 0>, rot=<0 0 0 1>}
 HandAttachLinFloorSpace {type=Curve, loc=<0 0 0>, rot=<0 0 0 1>}
 HandAttachRinFloorSpace {type=Curve, loc=<0 0 0>, rot=<0 0 0 1>}
 CanBypassSlideInitiate {type=Curve, loc=<0 0 0>, rot=<0 0 0 1>}
 AttachEmote1InPelvisSpace {type=Curve, loc=<0 0 0>, rot=<0 0 0 1>}
 AttachEmote2InPelvisSpace {type=Curve, loc=<0 0 0>, rot=<0 0 0 1>}
 AttachEmote3InPelvisSpace {type=Curve, loc=<0 0 0>, rot=<0 0 0 1>}
 LeftLegFK {type=Curve, loc=<0 0 0>, rot=<0 0 0 1>}
 RightLegFK {type=Curve, loc=<0 0 0>, rot=<0 0 0 1>}
 Disable_JamV2_FootShiftAdditive {type=Curve, loc=<0 0 0>, rot=<0 0 0 1>}
 Disable_JamV2_SpineMotionAdditive {type=Curve, loc=<0 0 0>, rot=<0 0 0 1>}
 Disable_JamV2_HeadMotionReplace {type=Curve, loc=<0 0 0>, rot=<0 0 0 1>}
 Disable_JamV2_LeftArmReplace {type=Curve, loc=<0 0 0>, rot=<0 0 0 1>}
 Disable_JamV2_RightArmReplace {type=Curve, loc=<0 0 0>, rot=<0 0 0 1>}
 CTRL_expressions_browInnerInL {type=Curve, loc=<0 0 0>, rot=<0 0 0 1>}
 CTRL_expressions_browInnerInR {type=Curve, loc=<0 0 0>, rot=<0 0 0 1>}
 CTRL_expressions_browInnerMidDownL {type=Curve, loc=<0 0 0>, rot=<0 0 0 1>}
 CTRL_expressions_browInnerMidDownR {type=Curve, loc=<0 0 0>, rot=<0 0 0 1>}
 CTRL_expressions_browInnerMidInL {type=Curve, loc=<0 0 0>, rot=<0 0 0 1>}
 CTRL_expressions_browInnerMidInR {type=Curve, loc=<0 0 0>, rot=<0 0 0 1>}
 CTRL_expressions_browInnerMidOutL {type=Curve, loc=<0 0 0>, rot=<0 0 0 1>}
 CTRL_expressions_browInnerMidOutR {type=Curve, loc=<0 0 0>, rot=<0 0 0 1>}
 CTRL_expressions_browInnerMidUpL {type=Curve, loc=<0 0 0>, rot=<0 0 0 1>}
 CTRL_expressions_browInnerMidUpR {type=Curve, loc=<0 0 0>, rot=<0 0 0 1>}
 CTRL_expressions_browInnerOutL {type=Curve, loc=<0 0 0>, rot=<0 0 0 1>}
 CTRL_expressions_browInnerOutR {type=Curve, loc=<0 0 0>, rot=<0 0 0 1>}
 CTRL_expressions_browOuterMidDownL {type=Curve, loc=<0 0 0>, rot=<0 0 0 1>}
 CTRL_expressions_browOuterMidDownR {type=Curve, loc=<0 0 0>, rot=<0 0 0 1>}
 CTRL_expressions_browOuterMidInL {type=Curve, loc=<0 0 0>, rot=<0 0 0 1>}
 CTRL_expressions_browOuterMidInR {type=Curve, loc=<0 0 0>, rot=<0 0 0 1>}
 CTRL_expressions_browOuterMidOutL {type=Curve, loc=<0 0 0>, rot=<0 0 0 1>}
 CTRL_expressions_browOuterMidOutR {type=Curve, loc=<0 0 0>, rot=<0 0 0 1>}
 CTRL_expressions_browOuterMidUpL {type=Curve, loc=<0 0 0>, rot=<0 0 0 1>}
 CTRL_expressions_browOuterMidUpR {type=Curve, loc=<0 0 0>, rot=<0 0 0 1>}
 CTRL_expressions_browRollNegInL {type=Curve, loc=<0 0 0>, rot=<0 0 0 1>}
 CTRL_expressions_browRollNegInR {type=Curve, loc=<0 0 0>, rot=<0 0 0 1>}
 CTRL_expressions_browRollNegOutL {type=Curve, loc=<0 0 0>, rot=<0 0 0 1>}
 CTRL_expressions_browRollNegOutR {type=Curve, loc=<0 0 0>, rot=<0 0 0 1>}
 CTRL_expressions_browRollPosInL {type=Curve, loc=<0 0 0>, rot=<0 0 0 1>}
 CTRL_expressions_browRollPosInR {type=Curve, loc=<0 0 0>, rot=<0 0 0 1>}
 CTRL_expressions_browRollPosOutL {type=Curve, loc=<0 0 0>, rot=<0 0 0 1>}
 CTRL_expressions_browRollPosOutR {type=Curve, loc=<0 0 0>, rot=<0 0 0 1>}
 CTRL_expressions_eyeFaceScrunchL {type=Curve, loc=<0 0 0>, rot=<0 0 0 1>}
 CTRL_expressions_eyeFaceScrunchR {type=Curve, loc=<0 0 0>, rot=<0 0 0 1>}
 CTRL_expressions_eyePupilNarrowL {type=Curve, loc=<0 0 0>, rot=<0 0 0 1>}
 CTRL_expressions_eyePupilNarrowR {type=Curve, loc=<0 0 0>, rot=<0 0 0 1>}
 CTRL_expressions_eyePupilWideL {type=Curve, loc=<0 0 0>, rot=<0 0 0 1>}
 CTRL_expressions_eyePupilWideR {type=Curve, loc=<0 0 0>, rot=<0 0 0 1>}
 CTRL_expressions_mouthLipsPressL {type=Curve, loc=<0 0 0>, rot=<0 0 0 1>}
 CTRL_expressions_mouthLipsPressR {type=Curve, loc=<0 0 0>, rot=<0 0 0 1>}
 CTRL_expressions_mouthLipsPullDL {type=Curve, loc=<0 0 0>, rot=<0 0 0 1>}
 CTRL_expressions_mouthLipsPullDR {type=Curve, loc=<0 0 0>, rot=<0 0 0 1>}
 CTRL_expressions_mouthLipsPullUL {type=Curve, loc=<0 0 0>, rot=<0 0 0 1>}
 CTRL_expressions_mouthLipsPullUR {type=Curve, loc=<0 0 0>, rot=<0 0 0 1>}
 CTRL_expressions_mouthLipsPushDL {type=Curve, loc=<0 0 0>, rot=<0 0 0 1>}
 CTRL_expressions_mouthLipsPushDR {type=Curve, loc=<0 0 0>, rot=<0 0 0 1>}
 CTRL_expressions_mouthLipsPushUL {type=Curve, loc=<0 0 0>, rot=<0 0 0 1>}
 CTRL_expressions_mouthLipsPushUR {type=Curve, loc=<0 0 0>, rot=<0 0 0 1>}
 CTRL_expressions_mouthLipsThickDL {type=Curve, loc=<0 0 0>, rot=<0 0 0 1>}
 CTRL_expressions_mouthLipsThickDR {type=Curve, loc=<0 0 0>, rot=<0 0 0 1>}
 CTRL_expressions_mouthLipsThickInwardDL {type=Curve, loc=<0 0 0>, rot=<0 0 0 1>}
 CTRL_expressions_mouthLipsThickInwardDR {type=Curve, loc=<0 0 0>, rot=<0 0 0 1>}
 CTRL_expressions_mouthLipsThickInwardUL {type=Curve, loc=<0 0 0>, rot=<0 0 0 1>}
 CTRL_expressions_mouthLipsThickInwardUR {type=Curve, loc=<0 0 0>, rot=<0 0 0 1>}
 CTRL_expressions_mouthLipsThickUL {type=Curve, loc=<0 0 0>, rot=<0 0 0 1>}
 CTRL_expressions_mouthLipsThickUR {type=Curve, loc=<0 0 0>, rot=<0 0 0 1>}
 CTRL_expressions_mouthLipsThinDL {type=Curve, loc=<0 0 0>, rot=<0 0 0 1>}
 CTRL_expressions_mouthLipsThinDR {type=Curve, loc=<0 0 0>, rot=<0 0 0 1>}
 CTRL_expressions_mouthLipsThinInwardDL {type=Curve, loc=<0 0 0>, rot=<0 0 0 1>}
 CTRL_expressions_mouthLipsThinInwardDR {type=Curve, loc=<0 0 0>, rot=<0 0 0 1>}
 CTRL_expressions_mouthLipsThinInwardUL {type=Curve, loc=<0 0 0>, rot=<0 0 0 1>}
 CTRL_expressions_mouthLipsThinInwardUR {type=Curve, loc=<0 0 0>, rot=<0 0 0 1>}
 CTRL_expressions_mouthLipsThinUL {type=Curve, loc=<0 0 0>, rot=<0 0 0 1>}
 CTRL_expressions_mouthLipsThinUR {type=Curve, loc=<0 0 0>, rot=<0 0 0 1>}
 CTRL_expressions_mouthLipsTightenDL {type=Curve, loc=<0 0 0>, rot=<0 0 0 1>}
 CTRL_expressions_mouthLipsTightenDR {type=Curve, loc=<0 0 0>, rot=<0 0 0 1>}
 CTRL_expressions_mouthLipsTightenUL {type=Curve, loc=<0 0 0>, rot=<0 0 0 1>}
 CTRL_expressions_mouthLipsTightenUR {type=Curve, loc=<0 0 0>, rot=<0 0 0 1>}
 CTRL_expressions_mouthLowerLipRollInL {type=Curve, loc=<0 0 0>, rot=<0 0 0 1>}
 CTRL_expressions_mouthLowerLipRollInR {type=Curve, loc=<0 0 0>, rot=<0 0 0 1>}
 CTRL_expressions_mouthLowerLipRollOutL {type=Curve, loc=<0 0 0>, rot=<0 0 0 1>}
 CTRL_expressions_mouthLowerLipRollOutR {type=Curve, loc=<0 0 0>, rot=<0 0 0 1>}
 CTRL_expressions_mouthSharpCornerPullL {type=Curve, loc=<0 0 0>, rot=<0 0 0 1>}
 CTRL_expressions_mouthSharpCornerPullR {type=Curve, loc=<0 0 0>, rot=<0 0 0 1>}
 CTRL_expressions_mouthUpperLipRollInL {type=Curve, loc=<0 0 0>, rot=<0 0 0 1>}
 CTRL_expressions_mouthUpperLipRollInR {type=Curve, loc=<0 0 0>, rot=<0 0 0 1>}
 CTRL_expressions_mouthUpperLipRollOutL {type=Curve, loc=<0 0 0>, rot=<0 0 0 1>}
 CTRL_expressions_mouthUpperLipRollOutR {type=Curve, loc=<0 0 0>, rot=<0 0 0 1>}
 CTRL_expressions_noseNasolabialDeepenL {type=Curve, loc=<0 0 0>, rot=<0 0 0 1>}
 CTRL_expressions_noseNasolabialDeepenR {type=Curve, loc=<0 0 0>, rot=<0 0 0 1>}
 CTRL_expressions_tongueBendDown {type=Curve, loc=<0 0 0>, rot=<0 0 0 1>}
 CTRL_expressions_tongueBendUp {type=Curve, loc=<0 0 0>, rot=<0 0 0 1>}
 CTRL_expressions_tongueTwistLeft {type=Curve, loc=<0 0 0>, rot=<0 0 0 1>}
 CTRL_expressions_tongueTwistRight {type=Curve, loc=<0 0 0>, rot=<0 0 0 1>}
 WM_BrowLateral_L {type=Curve, loc=<0 0 0>, rot=<0 0 0 1>}
 WM_BrowLateral_Offset_L {type=Curve, loc=<0 0 0>, rot=<0 0 0 1>}
 WM_BrowLateral_Offset_R {type=Curve, loc=<0 0 0>, rot=<0 0 0 1>}
 WM_BrowLateral_R {type=Curve, loc=<0 0 0>, rot=<0 0 0 1>}
 WM_BrowRaiseInner_L {type=Curve, loc=<0 0 0>, rot=<0 0 0 1>}
 WM_BrowRaiseInner_Offset_L {type=Curve, loc=<0 0 0>, rot=<0 0 0 1>}
 WM_BrowRaiseInner_Offset_R {type=Curve, loc=<0 0 0>, rot=<0 0 0 1>}
 WM_BrowRaiseInner_R {type=Curve, loc=<0 0 0>, rot=<0 0 0 1>}
 WM_BrowRaiseOutter_L {type=Curve, loc=<0 0 0>, rot=<0 0 0 1>}
 WM_BrowRaiseOutter_Offset_L {type=Curve, loc=<0 0 0>, rot=<0 0 0 1>}
 WM_BrowRaiseOutter_Offset_R {type=Curve, loc=<0 0 0>, rot=<0 0 0 1>}
 WM_BrowRaiseOutter_R {type=Curve, loc=<0 0 0>, rot=<0 0 0 1>}
 WM_CheekRaise_L {type=Curve, loc=<0 0 0>, rot=<0 0 0 1>}
 WM_CheekRaise_Offset_L {type=Curve, loc=<0 0 0>, rot=<0 0 0 1>}
 WM_CheekRaise_Offset_R {type=Curve, loc=<0 0 0>, rot=<0 0 0 1>}
 WM_CheekRaise_R {type=Curve, loc=<0 0 0>, rot=<0 0 0 1>}
 WM_Lips_L {type=Curve, loc=<0 0 0>, rot=<0 0 0 1>}
 WM_Lips_Offset_L {type=Curve, loc=<0 0 0>, rot=<0 0 0 1>}
 WM_Lips_Offset_R {type=Curve, loc=<0 0 0>, rot=<0 0 0 1>}
 WM_Lips_R {type=Curve, loc=<0 0 0>, rot=<0 0 0 1>}
 WM_MouthStretch_L {type=Curve, loc=<0 0 0>, rot=<0 0 0 1>}
 WM_MouthStretch_Offset_L {type=Curve, loc=<0 0 0>, rot=<0 0 0 1>}
 WM_MouthStretch_Offset_R {type=Curve, loc=<0 0 0>, rot=<0 0 0 1>}
 WM_MouthStretch_R {type=Curve, loc=<0 0 0>, rot=<0 0 0 1>}
 WM_NoseWrinkle_L {type=Curve, loc=<0 0 0>, rot=<0 0 0 1>}
 WM_NoseWrinkle_Offset_L {type=Curve, loc=<0 0 0>, rot=<0 0 0 1>}
 WM_NoseWrinkle_Offset_R {type=Curve, loc=<0 0 0>, rot=<0 0 0 1>}
 WM_NoseWrinkle_R {type=Curve, loc=<0 0 0>, rot=<0 0 0 1>}
 WM_Smile_L {type=Curve, loc=<0 0 0>, rot=<0 0 0 1>}
 WM_Smile_Offset_L {type=Curve, loc=<0 0 0>, rot=<0 0 0 1>}
 WM_Smile_Offset_R {type=Curve, loc=<0 0 0>, rot=<0 0 0 1>}
 WM_Smile_R {type=Curve, loc=<0 0 0>, rot=<0 0 0 1>}
 root {loc=<0 0 0>, rot=<0 0 0 1>}
 pelvis {parent=root, loc=<0 2.2809 95.8968>, rot=<-0.0224 0.7068 0.0224 -0.7068>}
 spine_01 {parent=pelvis, loc=<3.6771 0 0>, rot=<0 0 -0.1258 -0.9921>}
 spine_02 {parent=spine_01, loc=<6.7951 0 0>, rot=<0 0 0.0302 -0.9995>}
 spine_03 {parent=spine_02, loc=<7.2382 0 0>, rot=<0 0 0.0954 -0.9954>}
 spine_04 {parent=spine_03, loc=<8.5239 0 0>, rot=<0 0 0.0512 -0.9987>}
 spine_05 {parent=spine_04, loc=<19.4398 0 0>, rot=<0 0 0.0059 -1>}
 clavicle_l {parent=spine_05, loc=<5.5163 -1.3148 -1.4279>, rot=<0.0833 -0.7546 -0.0413 -0.6496>}
 upperarm_l {parent=clavicle_l, loc=<17.8095 0 0>, rot=<-0.0199 -0.3917 -0.0202 -0.9196>}
 lowerarm_l {parent=upperarm_l, loc=<27.7711 0 0>, rot=<0 0 -0.3336 -0.9427>}
 hand_l {parent=lowerarm_l, loc=<27.2511 0 0>, rot=<-0.5576 -0.0017 0.0206 -0.8299>}
 index_metacarpal_l {parent=hand_l, loc=<3.4445 0.3847 -2.3793>, rot=<0.029 -0.0637 -0.0071 -0.9975>}
 index_01_l {parent=index_metacarpal_l, loc=<5.8771 -0.0432 0.2409>, rot=<0 0 0.2026 -0.9793>}
 index_02_l {parent=index_01_l, loc=<4.08 0 0>, rot=<0 0 0.1296 -0.9916>}
 index_03_l {parent=index_02_l, loc=<2.595 0 0>, rot=<0 0 0.109 -0.994>}
 middle_metacarpal_l {parent=hand_l, loc=<3.3758 0.7536 -0.1829>, rot=<-0.0373 0.0004 -0.0203 -0.9991>}
 middle_01_l {parent=middle_metacarpal_l, loc=<6.0982 0 0>, rot=<0 0 0.2721 -0.9623>}
 middle_02_l {parent=middle_01_l, loc=<5.169 0 0>, rot=<0 0 0.1803 -0.9836>}
 middle_03_l {parent=middle_02_l, loc=<2.474 0 0>, rot=<0 0 0.0872 -0.9962>}
 pinky_metacarpal_l {parent=hand_l, loc=<3.3144 0.3059 2.3911>, rot=<-0.2182 0.1882 0.0583 -0.9558>}
 pinky_01_l {parent=pinky_metacarpal_l, loc=<4.9576 0.1431 -0.1988>, rot=<0.09 -0.017 0.1281 -0.9875>}
 pinky_02_l {parent=pinky_01_l, loc=<3.816 0 0>, rot=<0 0 0.1847 -0.9828>}
 pinky_03_l {parent=pinky_02_l, loc=<2.04 0 0>, rot=<0 0 0.0429 -0.9991>}
 ring_metacarpal_l {parent=hand_l, loc=<3.3743 0.5425 1.0918>, rot=<-0.1166 0.1006 -0.0257 -0.9877>}
 ring_01_l {parent=ring_metacarpal_l, loc=<5.6455 0.0416 -0.0207>, rot=<0.0542 -0.0132 0.2535 -0.9657>}
 ring_02_l {parent=ring_01_l, loc=<4.977 0 0>, rot=<0 0 0.1647 -0.9863>}
 ring_03_l {parent=ring_02_l, loc=<2.265 0 0>, rot=<0 0 0.0799 -0.9968>}
 thumb_01_l {parent=hand_l, loc=<1.9924 -1.3566 -2.5815>, rot=<0.5052 -0.3691 -0.067 -0.7772>}
 thumb_02_l {parent=thumb_01_l, loc=<4.378 0 0>, rot=<0.0336 0.0103 0.2019 -0.9788>}
 thumb_03_l {parent=thumb_02_l, loc=<3.086 0 0>, rot=<0 0 0.0872 -0.9962>}
 hand_attach_l {parent=hand_l, loc=<8.5 -3 0>, rot=<0.7071 0.7071 0 0>}
 weapon_l {parent=hand_l, loc=<-0.1974 -3.4111 0.3811>, rot=<0 0 0 1>}
 lowerarm_twist_01_l {parent=lowerarm_l, loc=<18.1674 0 0>, rot=<0.0115 0.0149 0.0175 -0.9997>}
 lowerarm_twist_02_l {parent=lowerarm_l, loc=<9.0837 0 0>, rot=<0.0115 0.0149 0.0175 -0.9997>}
 upperarm_twist_01_l {parent=upperarm_l, loc=<9.257 0 0>, rot=<-0.0001 0.0188 -0.0028 -0.9998>}
 upperarm_twist_02_l {parent=upperarm_l, loc=<18.5141 0 0>, rot=<0 0 0 1>}
 clavicle_r {parent=spine_05, loc=<5.5162 -1.3148 1.4279>, rot=<0.7546 0.0833 -0.6496 0.0413>}
 upperarm_r {parent=clavicle_r, loc=<-17.8096 0 0.0004>, rot=<-0.0199 -0.3917 -0.0202 -0.9196>}
 lowerarm_r {parent=upperarm_r, loc=<-27.7707 0 0>, rot=<0 0 -0.3336 -0.9427>}
 hand_r {parent=lowerarm_r, loc=<-27.251 0 0>, rot=<-0.5576 -0.0017 0.0206 -0.8299>}
 index_metacarpal_r {parent=hand_r, loc=<-3.4445 -0.3852 2.3793>, rot=<0.029 -0.0637 -0.0071 -0.9975>}
 index_01_r {parent=index_metacarpal_r, loc=<-5.8772 0.0434 -0.241>, rot=<0 0 0.2026 -0.9793>}
 index_02_r {parent=index_01_r, loc=<-4.0799 0 0>, rot=<0 0 0.1296 -0.9916>}
 index_03_r {parent=index_02_r, loc=<-2.5951 0 0>, rot=<0 0 0.109 -0.994>}
 middle_metacarpal_r {parent=hand_r, loc=<-3.3758 -0.754 0.1828>, rot=<-0.0373 0.0004 -0.0203 -0.9991>}
 middle_01_r {parent=middle_metacarpal_r, loc=<-6.0984 0.0001 0>, rot=<0 0 0.2721 -0.9623>}
 middle_02_r {parent=middle_01_r, loc=<-5.169 0.0001 0>, rot=<0 0 0.1803 -0.9836>}
 middle_03_r {parent=middle_02_r, loc=<-2.474 0 0>, rot=<0 0 0.0872 -0.9962>}
 pinky_metacarpal_r {parent=hand_r, loc=<-3.3147 -0.3059 -2.3913>, rot=<-0.2182 0.1882 0.0583 -0.9558>}
 pinky_01_r {parent=pinky_metacarpal_r, loc=<-4.9573 -0.1433 0.1989>, rot=<0.09 -0.017 0.1281 -0.9875>}
 pinky_02_r {parent=pinky_01_r, loc=<-3.816 0 0>, rot=<0 0 0.1847 -0.9828>}
 pinky_03_r {parent=pinky_02_r, loc=<-2.04 0 0>, rot=<0 0 0.0429 -0.9991>}
 ring_metacarpal_r {parent=hand_r, loc=<-3.3742 -0.543 -1.0918>, rot=<-0.1166 0.1006 -0.0257 -0.9877>}
 ring_01_r {parent=ring_metacarpal_r, loc=<-5.6457 -0.0414 0.0207>, rot=<0.0542 -0.0132 0.2535 -0.9657>}
 ring_02_r {parent=ring_01_r, loc=<-4.9771 0 0>, rot=<0 0 0.1647 -0.9863>}
 ring_03_r {parent=ring_02_r, loc=<-2.265 -0.0001 0>, rot=<0 0 0.0799 -0.9968>}
 thumb_01_r {parent=hand_r, loc=<-1.9928 1.3567 2.5813>, rot=<0.5052 -0.3691 -0.067 -0.7772>}
 thumb_02_r {parent=thumb_01_r, loc=<-4.3778 -0.0004 -0.0001>, rot=<0.0336 0.0103 0.2019 -0.9788>}
 thumb_03_r {parent=thumb_02_r, loc=<-3.086 0 0>, rot=<0 0 0.0872 -0.9962>}
 weapon_r {parent=hand_r, loc=<0.1975 3.4112 -0.3811>, rot=<0 0 0 1>}
 hand_attach_r {parent=hand_r, loc=<-8.5 3 0>, rot=<0 0 0.7071 0.7071>}
 lowerarm_twist_01_r {parent=lowerarm_r, loc=<-18.1673 0 0>, rot=<0.0115 0.0149 0.0175 -0.9997>}
 lowerarm_twist_02_r {parent=lowerarm_r, loc=<-9.0837 0 0>, rot=<0.0115 0.0149 0.0175 -0.9997>}
 upperarm_twist_01_r {parent=upperarm_r, loc=<-9.2569 0 -0.0001>, rot=<-0.0001 0.0188 -0.0028 -0.9998>}
 upperarm_twist_02_r {parent=upperarm_r, loc=<-18.5138 0 -0.0002>, rot=<0 0 0 1>}
 neck_01 {parent=spine_05, loc=<11.8878 0 0>, rot=<0 0 -0.2073 -0.9783>}
 neck_02 {parent=neck_01, loc=<5.1103 0 0>, rot=<0 0 0.0167 -0.9999>}
 head {parent=neck_02, loc=<4.913 0 0>, rot=<0 0 0.1035 -0.9946>}
 attach_cape {parent=spine_05, loc=<1.1827 -16.4319 0>, rot=<0.1003 -0.7 -0.1003 -0.7>}
 attach_backpack {parent=spine_05, loc=<-7.7085 -19.1852 0>, rot=<0.0297 -0.7065 -0.0297 -0.7065>}
 thigh_l {parent=pelvis, loc=<-2.3657 0.11 -9.9692>, rot=<0.0724 0.0295 -0.029 -0.9965>}
 calf_l {parent=thigh_l, loc=<-43.3413 0 0>, rot=<0 0 -0.0437 -0.999>}
 foot_l {parent=calf_l, loc=<-42.2179 0 0>, rot=<-0.0007 -0.0269 0.0232 -0.9994>}
 ball_l {parent=foot_l, loc=<-7.0094 15.2376 0.5389>, rot=<0 0 -0.7071 -0.7071>}
 foot_l_inner {parent=ball_l, type=Null, loc=<0 1.2 5>, rot=<0 0 0 1>}
 foot_l_outer {parent=ball_l, type=Null, loc=<0 1.2 -6.5>, rot=<0 0 0 1>}
 foot_l_heel {parent=ball_l, type=Null, loc=<-20 1.2 0>, rot=<0 0 0 1>}
 foot_l_toe_tip {parent=ball_l, type=Null, loc=<7.5 1.2 0>, rot=<0 0 0 1>}
 calf_twist_01_l {parent=calf_l, loc=<-28.1453 0 -0.0754>, rot=<-0.0001 -0.0023 0.0232 -0.9997>}
 calf_twist_02_l {parent=calf_l, loc=<-14.0726 0 0>, rot=<-0.0001 -0.0023 0.0232 -0.9997>}
 thigh_twist_01_l {parent=thigh_l, loc=<-14.4471 0 0>, rot=<0 -0.0023 -0.0114 -0.9999>}
 thigh_twist_02_l {parent=thigh_l, loc=<-28.8942 0 0>, rot=<0 -0.0023 -0.0114 -0.9999>}
 thigh_r {parent=pelvis, loc=<-2.3657 0.1195 9.9691>, rot=<0.0295 -0.0724 0.9965 -0.029>}
 calf_r {parent=thigh_r, loc=<43.3413 0 0>, rot=<0 0 -0.0437 -0.999>}
 foot_r {parent=calf_r, loc=<42.2179 0 0>, rot=<-0.0007 -0.0269 0.0232 -0.9994>}
 ball_r {parent=foot_r, loc=<7.0094 -15.2376 -0.5389>, rot=<0 0 -0.7071 -0.7071>}
 foot_r_toe_tip {parent=ball_r, type=Null, loc=<-7.5 0 0>, rot=<0 0 0 1>}
 foot_r_heel {parent=ball_r, type=Null, loc=<20 -1.2 0>, rot=<0 0 0 1>}
 foot_r_outer {parent=ball_r, type=Null, loc=<0 0 6.5>, rot=<0 0 0 1>}
 foot_r_inner {parent=ball_r, type=Null, loc=<0 0 -5>, rot=<0 0 0 1>}
 calf_twist_01_r {parent=calf_r, loc=<28.1453 0 0.0755>, rot=<-0.0001 -0.0023 0.0232 -0.9997>}
 calf_twist_02_r {parent=calf_r, loc=<14.0726 0 0.0377>, rot=<-0.0001 -0.0023 0.0232 -0.9997>}
 thigh_twist_01_r {parent=thigh_r, loc=<14.4471 0 0>, rot=<0 -0.0023 -0.0114 -0.9999>}
 thigh_twist_02_r {parent=thigh_r, loc=<28.8942 0 0>, rot=<0 -0.0023 -0.0114 -0.9999>}
 attach {parent=root, loc=<0 0 0>, rot=<0 0 0 1>}
 attach_emote_01 {parent=root, loc=<0 0 0>, rot=<0 0 0 1>}
 attach_emote_02 {parent=root, loc=<0 0 0>, rot=<0 0 0 1>}
 attach_emote_03 {parent=root, loc=<0 0 0>, rot=<0 0 0 1>}
 warp_point_01 {parent=root, loc=<0 0 0>, rot=<0 0 0 1>}
 warp_point_02 {parent=root, loc=<0 0 0>, rot=<0 0 0 1>}
 warp_point_03 {parent=root, loc=<0 0 0>, rot=<0 0 0 1>}
 root_fp {parent=root, loc=<0 0 0>, rot=<0 0 0 1>}
 camera_root {parent=root_fp, loc=<0 0 0>, rot=<0 0 0 1>}
 camera {parent=camera_root, loc=<0 0 0>, rot=<0 0 0 1>}
 attach_fp {parent=root_fp, loc=<0 0 0>, rot=<0 0 0 1>}
 ik_foot_root {parent=root, loc=<0 0 0>, rot=<0 0 0 1>}
 ik_foot_l {parent=ik_foot_root, loc=<14.0877 -0.9943 8.2376>, rot=<0.0636 0.7042 0.0414 -0.7059>}
 ik_foot_r {parent=ik_foot_root, loc=<-14.0876 -0.9848 8.2382>, rot=<0.7059 0.0414 -0.7042 0.0636>}
 ik_hand_root {parent=root, loc=<0 0 0>, rot=<0 0 0 1>}
 ik_hand_gun {parent=ik_hand_root, loc=<-47.7683 15.6988 104.4664>, rot=<0.5985 0.0096 0.5206 -0.6089>}
 ik_hand_l {parent=ik_hand_gun, loc=<43.7336 61.6698 58.4081>, rot=<-0.4578 -0.6455 -0.6114 0>}
 ik_hand_r {parent=ik_hand_gun, loc=<0 0 0>, rot=<0 0 0 1>}
 interaction {parent=root, loc=<0 0 0>, rot=<0 0 0 1>}
 center_of_mass {parent=root, loc=<0 0 0>, rot=<0 0 0 1>}
 root/RootJoint {type=Connector, loc=<0 0 0>, rot=<0 0 0 1>}
 root/Body {type=Connector, loc=<0 0 0>, rot=<0 0 0 1>}
 Prop/Parent {type=Connector, loc=<0 0 0>, rot=<0 0 0 1>}
 Prop/Spaces {type=Connector, loc=<0 0 0>, rot=<0 0 0 1>}
 Prop/Control Vis Channel Host {type=Connector, loc=<0 0 0>, rot=<0 0 0 1>}
 Body/Parent {type=Connector, loc=<0 0 0>, rot=<0 0 0 1>}
 Body/Body {type=Connector, loc=<0 0 0>, rot=<0 0 0 1>}
 Body/Right Hip {type=Connector, loc=<0 0 0>, rot=<0 0 0 1>}
 Body/Left Hip {type=Connector, loc=<0 0 0>, rot=<0 0 0 1>}
 Spine/Start {type=Connector, loc=<0 0 0>, rot=<0 0 0 1>}
 Spine/End {type=Connector, loc=<0 0 0>, rot=<0 0 0 1>}
 Spine/End Parent {type=Connector, loc=<0 0 0>, rot=<0 0 0 1>}
 Spine/Start Parent {type=Connector, loc=<0 0 0>, rot=<0 0 0 1>}
 Spine/Start Orient Spaces {type=Connector, loc=<0 0 0>, rot=<0 0 0 1>}
 Spine/End Orient Spaces {type=Connector, loc=<0 0 0>, rot=<0 0 0 1>}
 Spine/Start Snap To {type=Connector, loc=<0 0 0>, rot=<0 0 0 1>}
 Leg L/Start {type=Connector, loc=<0 0 0>, rot=<0 0 0 1>}
 Leg L/Mid {type=Connector, loc=<0 0 0>, rot=<0 0 0 1>}
 Leg L/End {type=Connector, loc=<0 0 0>, rot=<0 0 0 1>}
 Leg L/Parent {type=Connector, loc=<0 0 0>, rot=<0 0 0 1>}
 Leg L/IK Spaces {type=Connector, loc=<0 0 0>, rot=<0 0 0 1>}
 Leg L/FK Spaces {type=Connector, loc=<0 0 0>, rot=<0 0 0 1>}
 Foot L/Foot Joint {type=Connector, loc=<0 0 0>, rot=<0 0 0 1>}
 Foot L/Ball Joint {type=Connector, loc=<0 0 0>, rot=<0 0 0 1>}
 Foot L/Toe Joints {type=Connector, loc=<0 0 0>, rot=<0 0 0 1>}
 Foot L/Toe Tip Pivot {type=Connector, loc=<0 0 0>, rot=<0 0 0 1>}
 Foot L/Heel Pivot {type=Connector, loc=<0 0 0>, rot=<0 0 0 1>}
 Foot L/Inner Bank Pivot {type=Connector, loc=<0 0 0>, rot=<0 0 0 1>}
 Foot L/Outer Bank Pivot {type=Connector, loc=<0 0 0>, rot=<0 0 0 1>}
 Leg Upper Twist L/Start {type=Connector, loc=<0 0 0>, rot=<0 0 0 1>}
 Leg Upper Twist L/Parent {type=Connector, loc=<0 0 0>, rot=<0 0 0 1>}
 Leg Upper Twist L/End {type=Connector, loc=<0 0 0>, rot=<0 0 0 1>}
 Leg Lower Twist L/Start {type=Connector, loc=<0 0 0>, rot=<0 0 0 1>}
 Leg Lower Twist L/Parent {type=Connector, loc=<0 0 0>, rot=<0 0 0 1>}
 Leg Lower Twist L/End {type=Connector, loc=<0 0 0>, rot=<0 0 0 1>}
 Leg R/Start {type=Connector, loc=<0 0 0>, rot=<0 0 0 1>}
 Leg R/Mid {type=Connector, loc=<0 0 0>, rot=<0 0 0 1>}
 Leg R/End {type=Connector, loc=<0 0 0>, rot=<0 0 0 1>}
 Leg R/Parent {type=Connector, loc=<0 0 0>, rot=<0 0 0 1>}
 Leg R/IK Spaces {type=Connector, loc=<0 0 0>, rot=<0 0 0 1>}
 Leg R/FK Spaces {type=Connector, loc=<0 0 0>, rot=<0 0 0 1>}
 Foot R/Foot Joint {type=Connector, loc=<0 0 0>, rot=<0 0 0 1>}
 Foot R/Ball Joint {type=Connector, loc=<0 0 0>, rot=<0 0 0 1>}
 Foot R/Toe Joints {type=Connector, loc=<0 0 0>, rot=<0 0 0 1>}
 Foot R/Toe Tip Pivot {type=Connector, loc=<0 0 0>, rot=<0 0 0 1>}
 Foot R/Heel Pivot {type=Connector, loc=<0 0 0>, rot=<0 0 0 1>}
 Foot R/Inner Bank Pivot {type=Connector, loc=<0 0 0>, rot=<0 0 0 1>}
 Foot R/Outer Bank Pivot {type=Connector, loc=<0 0 0>, rot=<0 0 0 1>}
 Leg Upper Twist R/Start {type=Connector, loc=<0 0 0>, rot=<0 0 0 1>}
 Leg Upper Twist R/Parent {type=Connector, loc=<0 0 0>, rot=<0 0 0 1>}
 Leg Upper Twist R/End {type=Connector, loc=<0 0 0>, rot=<0 0 0 1>}
 Leg Lower Twist R/Start {type=Connector, loc=<0 0 0>, rot=<0 0 0 1>}
 Leg Lower Twist R/Parent {type=Connector, loc=<0 0 0>, rot=<0 0 0 1>}
 Leg Lower Twist R/End {type=Connector, loc=<0 0 0>, rot=<0 0 0 1>}
 Clavicle L/Start {type=Connector, loc=<0 0 0>, rot=<0 0 0 1>}
 Clavicle L/Parent {type=Connector, loc=<0 0 0>, rot=<0 0 0 1>}
 Clavicle L/End {type=Connector, loc=<0 0 0>, rot=<0 0 0 1>}
 Clavicle L/Orient Spaces {type=Connector, loc=<0 0 0>, rot=<0 0 0 1>}
 Arm L/Start {type=Connector, loc=<0 0 0>, rot=<0 0 0 1>}
 Arm L/Mid {type=Connector, loc=<0 0 0>, rot=<0 0 0 1>}
 Arm L/End {type=Connector, loc=<0 0 0>, rot=<0 0 0 1>}
 Arm L/Parent {type=Connector, loc=<0 0 0>, rot=<0 0 0 1>}
 Arm L/IK Spaces {type=Connector, loc=<0 0 0>, rot=<0 0 0 1>}
 Arm L/FK Spaces {type=Connector, loc=<0 0 0>, rot=<0 0 0 1>}
 Meta L/Parent {type=Connector, loc=<0 0 0>, rot=<0 0 0 1>}
 Meta L/Bones {type=Connector, loc=<0 0 0>, rot=<0 0 0 1>}
 Meta L/Spaces {type=Connector, loc=<0 0 0>, rot=<0 0 0 1>}
 Meta L/Orient Spaces {type=Connector, loc=<0 0 0>, rot=<0 0 0 1>}
 Meta L/Override Parents {type=Connector, loc=<0 0 0>, rot=<0 0 0 1>}
 Meta Spread Inner L/Parent {type=Connector, loc=<0 0 0>, rot=<0 0 0 1>}
 Meta Spread Inner L/Driven Controls {type=Connector, loc=<0 0 0>, rot=<0 0 0 1>}
 Meta Spread Inner L/Snap To {type=Connector, loc=<0 0 0>, rot=<0 0 0 1>}
 Meta Spread Outer L/Parent {type=Connector, loc=<0 0 0>, rot=<0 0 0 1>}
 Meta Spread Outer L/Driven Controls {type=Connector, loc=<0 0 0>, rot=<0 0 0 1>}
 Meta Spread Outer L/Snap To {type=Connector, loc=<0 0 0>, rot=<0 0 0 1>}
 Index L/Start {type=Connector, loc=<0 0 0>, rot=<0 0 0 1>}
 Index L/Parent {type=Connector, loc=<0 0 0>, rot=<0 0 0 1>}
 Index L/End {type=Connector, loc=<0 0 0>, rot=<0 0 0 1>}
 Index L/Orient Spaces {type=Connector, loc=<0 0 0>, rot=<0 0 0 1>}
 Middle L/Start {type=Connector, loc=<0 0 0>, rot=<0 0 0 1>}
 Middle L/Parent {type=Connector, loc=<0 0 0>, rot=<0 0 0 1>}
 Middle L/End {type=Connector, loc=<0 0 0>, rot=<0 0 0 1>}
 Middle L/Orient Spaces {type=Connector, loc=<0 0 0>, rot=<0 0 0 1>}
 Ring L/Start {type=Connector, loc=<0 0 0>, rot=<0 0 0 1>}
 Ring L/Parent {type=Connector, loc=<0 0 0>, rot=<0 0 0 1>}
 Ring L/End {type=Connector, loc=<0 0 0>, rot=<0 0 0 1>}
 Ring L/Orient Spaces {type=Connector, loc=<0 0 0>, rot=<0 0 0 1>}
 Pinky L/Start {type=Connector, loc=<0 0 0>, rot=<0 0 0 1>}
 Pinky L/Parent {type=Connector, loc=<0 0 0>, rot=<0 0 0 1>}
 Pinky L/End {type=Connector, loc=<0 0 0>, rot=<0 0 0 1>}
 Pinky L/Orient Spaces {type=Connector, loc=<0 0 0>, rot=<0 0 0 1>}
 Thumb L/Start {type=Connector, loc=<0 0 0>, rot=<0 0 0 1>}
 Thumb L/Parent {type=Connector, loc=<0 0 0>, rot=<0 0 0 1>}
 Thumb L/End {type=Connector, loc=<0 0 0>, rot=<0 0 0 1>}
 Thumb L/Orient Spaces {type=Connector, loc=<0 0 0>, rot=<0 0 0 1>}
 Finger Spread Inner L/Parent {type=Connector, loc=<0 0 0>, rot=<0 0 0 1>}
 Finger Spread Inner L/Driven Controls {type=Connector, loc=<0 0 0>, rot=<0 0 0 1>}
 Finger Spread Inner L/Snap To {type=Connector, loc=<0 0 0>, rot=<0 0 0 1>}
 Finger Spread Outer L/Parent {type=Connector, loc=<0 0 0>, rot=<0 0 0 1>}
 Finger Spread Outer L/Driven Controls {type=Connector, loc=<0 0 0>, rot=<0 0 0 1>}
 Finger Spread Outer L/Snap To {type=Connector, loc=<0 0 0>, rot=<0 0 0 1>}
 Finger Inner Curl L/Parent {type=Connector, loc=<0 0 0>, rot=<0 0 0 1>}
 Finger Inner Curl L/Driven Controls {type=Connector, loc=<0 0 0>, rot=<0 0 0 1>}
 Finger Inner Curl L/Snap To {type=Connector, loc=<0 0 0>, rot=<0 0 0 1>}
 Finger Outer Curl L/Parent {type=Connector, loc=<0 0 0>, rot=<0 0 0 1>}
 Finger Outer Curl L/Driven Controls {type=Connector, loc=<0 0 0>, rot=<0 0 0 1>}
 Finger Outer Curl L/Snap To {type=Connector, loc=<0 0 0>, rot=<0 0 0 1>}
 Arm Upper Twist L/Start {type=Connector, loc=<0 0 0>, rot=<0 0 0 1>}
 Arm Upper Twist L/Parent {type=Connector, loc=<0 0 0>, rot=<0 0 0 1>}
 Arm Upper Twist L/End {type=Connector, loc=<0 0 0>, rot=<0 0 0 1>}
 Arm Lower Twist L/Start {type=Connector, loc=<0 0 0>, rot=<0 0 0 1>}
 Arm Lower Twist L/Parent {type=Connector, loc=<0 0 0>, rot=<0 0 0 1>}
 Arm Lower Twist L/End {type=Connector, loc=<0 0 0>, rot=<0 0 0 1>}
 Clavicle R/Start {type=Connector, loc=<0 0 0>, rot=<0 0 0 1>}
 Clavicle R/Parent {type=Connector, loc=<0 0 0>, rot=<0 0 0 1>}
 Clavicle R/End {type=Connector, loc=<0 0 0>, rot=<0 0 0 1>}
 Clavicle R/Orient Spaces {type=Connector, loc=<0 0 0>, rot=<0 0 0 1>}
 Arm R/Start {type=Connector, loc=<0 0 0>, rot=<0 0 0 1>}
 Arm R/Mid {type=Connector, loc=<0 0 0>, rot=<0 0 0 1>}
 Arm R/End {type=Connector, loc=<0 0 0>, rot=<0 0 0 1>}
 Arm R/Parent {type=Connector, loc=<0 0 0>, rot=<0 0 0 1>}
 Arm R/IK Spaces {type=Connector, loc=<0 0 0>, rot=<0 0 0 1>}
 Arm R/FK Spaces {type=Connector, loc=<0 0 0>, rot=<0 0 0 1>}
 Meta R/Parent {type=Connector, loc=<0 0 0>, rot=<0 0 0 1>}
 Meta R/Bones {type=Connector, loc=<0 0 0>, rot=<0 0 0 1>}
 Meta R/Spaces {type=Connector, loc=<0 0 0>, rot=<0 0 0 1>}
 Meta R/Orient Spaces {type=Connector, loc=<0 0 0>, rot=<0 0 0 1>}
 Meta R/Override Parents {type=Connector, loc=<0 0 0>, rot=<0 0 0 1>}
 Meta Spread Inner R/Parent {type=Connector, loc=<0 0 0>, rot=<0 0 0 1>}
 Meta Spread Inner R/Driven Controls {type=Connector, loc=<0 0 0>, rot=<0 0 0 1>}
 Meta Spread Inner R/Snap To {type=Connector, loc=<0 0 0>, rot=<0 0 0 1>}
 Meta Spread Outer R/Parent {type=Connector, loc=<0 0 0>, rot=<0 0 0 1>}
 Meta Spread Outer R/Driven Controls {type=Connector, loc=<0 0 0>, rot=<0 0 0 1>}
 Meta Spread Outer R/Snap To {type=Connector, loc=<0 0 0>, rot=<0 0 0 1>}
 Index R/Start {type=Connector, loc=<0 0 0>, rot=<0 0 0 1>}
 Index R/Parent {type=Connector, loc=<0 0 0>, rot=<0 0 0 1>}
 Index R/End {type=Connector, loc=<0 0 0>, rot=<0 0 0 1>}
 Index R/Orient Spaces {type=Connector, loc=<0 0 0>, rot=<0 0 0 1>}
 Middle R/Start {type=Connector, loc=<0 0 0>, rot=<0 0 0 1>}
 Middle R/Parent {type=Connector, loc=<0 0 0>, rot=<0 0 0 1>}
 Middle R/End {type=Connector, loc=<0 0 0>, rot=<0 0 0 1>}
 Middle R/Orient Spaces {type=Connector, loc=<0 0 0>, rot=<0 0 0 1>}
 Ring R/Start {type=Connector, loc=<0 0 0>, rot=<0 0 0 1>}
 Ring R/Parent {type=Connector, loc=<0 0 0>, rot=<0 0 0 1>}
 Ring R/End {type=Connector, loc=<0 0 0>, rot=<0 0 0 1>}
 Ring R/Orient Spaces {type=Connector, loc=<0 0 0>, rot=<0 0 0 1>}
 Pinky R/Start {type=Connector, loc=<0 0 0>, rot=<0 0 0 1>}
 Pinky R/Parent {type=Connector, loc=<0 0 0>, rot=<0 0 0 1>}
 Pinky R/End {type=Connector, loc=<0 0 0>, rot=<0 0 0 1>}
 Pinky R/Orient Spaces {type=Connector, loc=<0 0 0>, rot=<0 0 0 1>}
 Thumb R/Start {type=Connector, loc=<0 0 0>, rot=<0 0 0 1>}
 Thumb R/Parent {type=Connector, loc=<0 0 0>, rot=<0 0 0 1>}
 Thumb R/End {type=Connector, loc=<0 0 0>, rot=<0 0 0 1>}
 Thumb R/Orient Spaces {type=Connector, loc=<0 0 0>, rot=<0 0 0 1>}
 Finger Spread Inner R/Parent {type=Connector, loc=<0 0 0>, rot=<0 0 0 1>}
 Finger Spread Inner R/Driven Controls {type=Connector, loc=<0 0 0>, rot=<0 0 0 1>}
 Finger Spread Inner R/Snap To {type=Connector, loc=<0 0 0>, rot=<0 0 0 1>}
 Finger Spread Outer R/Parent {type=Connector, loc=<0 0 0>, rot=<0 0 0 1>}
 Finger Spread Outer R/Driven Controls {type=Connector, loc=<0 0 0>, rot=<0 0 0 1>}
 Finger Spread Outer R/Snap To {type=Connector, loc=<0 0 0>, rot=<0 0 0 1>}
 Finger Inner Curl R/Parent {type=Connector, loc=<0 0 0>, rot=<0 0 0 1>}
 Finger Inner Curl R/Driven Controls {type=Connector, loc=<0 0 0>, rot=<0 0 0 1>}
 Finger Inner Curl R/Snap To {type=Connector, loc=<0 0 0>, rot=<0 0 0 1>}
 Finger Outer Curl R/Parent {type=Connector, loc=<0 0 0>, rot=<0 0 0 1>}
 Finger Outer Curl R/Driven Controls {type=Connector, loc=<0 0 0>, rot=<0 0 0 1>}
 Finger Outer Curl R/Snap To {type=Connector, loc=<0 0 0>, rot=<0 0 0 1>}
 Arm Upper Twist R/Start {type=Connector, loc=<0 0 0>, rot=<0 0 0 1>}
 Arm Upper Twist R/Parent {type=Connector, loc=<0 0 0>, rot=<0 0 0 1>}
 Arm Upper Twist R/End {type=Connector, loc=<0 0 0>, rot=<0 0 0 1>}
 Arm Lower Twist R/Start {type=Connector, loc=<0 0 0>, rot=<0 0 0 1>}
 Arm Lower Twist R/Parent {type=Connector, loc=<0 0 0>, rot=<0 0 0 1>}
 Arm Lower Twist R/End {type=Connector, loc=<0 0 0>, rot=<0 0 0 1>}
 Neck/Start {type=Connector, loc=<0 0 0>, rot=<0 0 0 1>}
 Neck/End {type=Connector, loc=<0 0 0>, rot=<0 0 0 1>}
 Neck/End Parent {type=Connector, loc=<0 0 0>, rot=<0 0 0 1>}
 Neck/Start Parent {type=Connector, loc=<0 0 0>, rot=<0 0 0 1>}
 Neck/Start Orient Spaces {type=Connector, loc=<0 0 0>, rot=<0 0 0 1>}
 Neck/End Orient Spaces {type=Connector, loc=<0 0 0>, rot=<0 0 0 1>}
 Neck/Start Snap To {type=Connector, loc=<0 0 0>, rot=<0 0 0 1>}
 Attach/Parent {type=Connector, loc=<0 0 0>, rot=<0 0 0 1>}
 Attach/Bones {type=Connector, loc=<0 0 0>, rot=<0 0 0 1>}
 Attach/Spaces {type=Connector, loc=<0 0 0>, rot=<0 0 0 1>}
 Attach/Orient Spaces {type=Connector, loc=<0 0 0>, rot=<0 0 0 1>}
 Attach/Override Parents {type=Connector, loc=<0 0 0>, rot=<0 0 0 1>}
 Stretch Feedback/Root {type=Connector, loc=<0 0 0>, rot=<0 0 0 1>}
 Stretch Feedback/Spine Elements {type=Connector, loc=<0 0 0>, rot=<0 0 0 1>}
 Stretch Feedback/Arm Elements {type=Connector, loc=<0 0 0>, rot=<0 0 0 1>}
 Stretch Feedback/Leg Elements {type=Connector, loc=<0 0 0>, rot=<0 0 0 1>}
 Stretch Feedback/Vis Channel Control {type=Connector, loc=<0 0 0>, rot=<0 0 0 1>}
 FPS Controls/Root {type=Connector, loc=<0 0 0>, rot=<0 0 0 1>}
 Emote Attach/Parent {type=Connector, loc=<0 0 0>, rot=<0 0 0 1>}
 Emote Attach/Bones {type=Connector, loc=<0 0 0>, rot=<0 0 0 1>}
 Emote Attach/Spaces {type=Connector, loc=<0 0 0>, rot=<0 0 0 1>}
 Emote Attach/Orient Spaces {type=Connector, loc=<0 0 0>, rot=<0 0 0 1>}
 Emote Attach/Override Parents {type=Connector, loc=<0 0 0>, rot=<0 0 0 1>}
 Warp Points/Parent {type=Connector, loc=<0 0 0>, rot=<0 0 0 1>}
 Warp Points/Bones {type=Connector, loc=<0 0 0>, rot=<0 0 0 1>}
 Warp Points/Spaces {type=Connector, loc=<0 0 0>, rot=<0 0 0 1>}
 Warp Points/Orient Spaces {type=Connector, loc=<0 0 0>, rot=<0 0 0 1>}
 Warp Points/Override Parents {type=Connector, loc=<0 0 0>, rot=<0 0 0 1>}
 IK Bone Pins/Root {type=Connector, loc=<0 0 0>, rot=<0 0 0 1>}
 IK Bone Pins/Drivers {type=Connector, loc=<0 0 0>, rot=<0 0 0 1>}
 IK Bone Pins/Driven {type=Connector, loc=<0 0 0>, rot=<0 0 0 1>}
]
variables [
 Arm_L_RotateOrder {type=EEulerRotationOrder, default=XYZ}
 Arm_R_RotateOrder {type=EEulerRotationOrder, default=XYZ}
]
